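# Niagara system for the Luminence skydiving contrail cosmetic, drawing glowing mesh effects around the diving or gliding player.
# System modules supply dive, dome direction and skeletal mesh values that, with glide lean and front-end preview inputs, drive the emitters.
# Card_Dome001 rings a bone with bobbing, outward-facing cards that scale with dive, while Card_Stream001 throws off drifting cards faster during lean.
# The Card_Dome, Card_Stream and Center001 emitters are disabled.
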NiagaraSystem /BRCosmetics/Athena/Effects/Particles/Skydiving/NS_Contrail_Luminence/NS_Contrail_Luminence
dependencies {
 usedDataInterfaces [
  &NiagaraDataInterfaceArrayFloat3
  &NiagaraDataInterfaceArrayFloat
  &NiagaraDataInterfaceColorCurve
  &NiagaraDataInterfaceCurve
  &NiagaraDataInterfaceVectorCurve
  &NiagaraDataInterfaceVector2DCurve
  &NiagaraDataInterfaceMeshRendererInfo
  &NiagaraDataInterfaceParticleRead
  &NiagaraDataInterfaceSocketReader
 ]
 usedDynamicInputs [&SimulationPosition, /Niagara/DynamicInputs/Lerp/SmoothLerpOverTime_Float]
 usedModules [
  &self:SystemSpawnScript.SystemScriptSource.SystemScriptGraph.NiagaraNodeAssignment_0.SetVariables_#1
  &self:SystemSpawnScript.SystemScriptSource.SystemScriptGraph.NiagaraNodeAssignment_2.SetVariables_#2
  &self:SystemSpawnScript.SystemScriptSource.SystemScriptGraph.NiagaraNodeAssignment_4.SetVariables_#3
  &self:SystemSpawnScript.SystemScriptSource.SystemScriptGraph.NiagaraNodeAssignment_7.SetVariables_#4
  &SystemState
  /Game/Effects/Niagara/Modules/Contrails/FN_Contrail_Setup
  &self:SystemSpawnScript.SystemScriptSource.SystemScriptGraph.NiagaraNodeAssignment_3.SetVariables_#5
  /Game/Effects/Niagara/Modules/Skydiving/FN_DivingAttributes/v5/FN_DivingAttributes
  &self:SystemSpawnScript.SystemScriptSource.SystemScriptGraph.NiagaraNodeAssignment_1.SetVariables_#6
  &FN_Integrator
  &self:SystemSpawnScript.SystemScriptSource.SystemScriptGraph.NiagaraNodeAssignment_5.SetVariables_#7
  &self:SystemSpawnScript.SystemScriptSource.SystemScriptGraph.NiagaraNodeAssignment_6.SetVariables_#8
 ]
}
emitters {
 Card_Dome {
  bEnabled=false
  emitterUpdateScript [
   &EmitterState(Life Cycle Mode=::NewEnumerator1, Loop Behavior=::NewEnumerator1, Loop Duration Mode=::NewEnumerator1, UseLoopDelay=false)
   &SpawnBurst_Instantaneous(Spawn Count=9, Use Spawn Probability=false, UseLoopCountLimit=false)
   &SpawnRate(SpawnRate=120, Use Spawn Probability=false) {enabled=false}
  ]
  particleSpawnScript [
   &InitializeParticle(
    Color Mode=::NewEnumerator0
    Lifetime=5
    Mesh Scale Mode=::NewEnumerator1
    Mesh Uniform Scale Max=3.5
    Mesh Uniform Scale Min=2.5
    Sprite Size=<70 90>
    Sprite Size Mode=::NewEnumerator4
    Sprite UV Mode=::NewEnumerator1
    UsePositionOffset=false
    Write Material Random=true
   )
   &SystemLocation(Offset=&RotateVectorByQuaternion(
     Vector To Rotate=/Niagara/DynamicInputs/Vectors/RotateVector(
      Vector To Rotate:Vector3f=&MakeVector(X=$System.DomeRadius, Y=0, Z=0)
      Yaw:NiagaraFloat=&ReturnNormalizedExecIndex(Normalized Index Scale=720, Normalized Index Includes One=false)
      Pitch:NiagaraFloat=&RandomRangeFloat(Minimum=-45, Maximum=0)
      Roll:NiagaraFloat=0
      Rotation Rate:NiagaraFloat=1
     )
     Quaternion=&MakeQuaternion(
      Quaternion From=3
      Axis=<1 0 0>
      Angle=0
      From=<0 0 1>
      To=$System.DomeDirecrion
      Stabilization Mode:ENiagara_OrientVectorCount=::NewEnumerator1
      Side Direction:Vector3f=?
      Up Direction:Vector3f=<0 0 1>
      Angle Type=::NewEnumerator0
      Coordinate Space=::Local
      Normalize:NiagaraBool=true
      Rotation Matrix=Linked() {linkedVariable {name=Engine.Owner.SystemLocalToWorldNoScale, type=NiagaraFloat}}
     )
    )) {enabled=false}
   /SetVariables(Particles.Position=&AddVectorToPosition(Position=$Particles.Position, Vector=<0 0 110>)) {enabled=false, isSetParameters=true, module=SetVariables_#9}
   &InitialMeshOrientation(
    Mesh Orientation Mode=::NewEnumerator2
    Orientation Vector=&Lerp_Vector(A=<7 0 -10>, B=<0 0 -10>, Alpha=$User.ParaGlideLeanAlpha)
    Use Rotation Vector=false
   )
   /SetVariables(Particles.CardIndex:NiagaraFloat=&RandomRangeFloat(Minimum=0, Maximum=3)) {isSetParameters=true, module=SetVariables_#10}
   &InitializeRibbon(
    Lifetime=0
    Position Mode:ENiagara_PositionInitializationMode=::NewEnumerator0
    Ribbon Facing Mode:ENiagara_UnsetDirectSet=::NewEnumerator1
    Ribbon Facing Vector:Vector3f=<1 0 0>
    Ribbon Width=50
    Write Material Random=true
   )
  ]
  particleUpdateScript [
   &SolveForcesAndVelocity(Clamp Velocity=false, Limit Acceleration=false)
   &ParticleState(Kill Particles When Lifetime Has Elapsed=false, Let Infinitely Lived Particles Die When Emitter Deactivates=true)
   &ScaleMeshSize() {enabled=false}
   &UpdateMeshOrientation(
    Facing Direction=&Lerp_Vector(A=<2 0 -0.12>, B=<1 0 0.34>, Alpha=$User.ParaGlideLeanAlpha)
    Facing Interpolation=true
    Facing Interpolation Rate=7
    Orientation Method=::NewEnumerator1
    Override Quat=false
    TransformFacing=false
    TransformUp=false
    Up Direction:Vector3f=<1 0 1>
   ) {enabled=false, module=UpdateMeshOrientation001}
   &SkeletalMeshLocation(
    Bone / Socket Index=4
    Bone Sampling Mode=::NewEnumerator4
    MaskSpawnGroup=false
    Orientation Sampling=::NewEnumerator1
    Sampled Position Offset=&RotateVectorByQuaternion(
     Vector To Rotate=&RotateVector(
      Vector To Rotate:Vector3f=&MakeVector(X=&Lerp_Float(A=$System.DomeRadius, B=200, Alpha=$User.ParaGlideLeanAlpha), Y=0, Z=0)
      Yaw:NiagaraFloat=&ReturnNormalizedExecIndex(Normalized Index Scale=720, Normalized Index Includes One=false)
      Pitch:NiagaraFloat=&RandomRangeFloat(Minimum=0, Maximum=0)
      Roll:NiagaraFloat=0
      Rotation Rate:NiagaraFloat=1
     )
     Quaternion=&MakeQuaternion(
      Quaternion From=3
      Axis=<1 0 0>
      Angle=0
      From=<0 0 1>
      To=$System.DomeDirecrion
      Stabilization Mode:ENiagara_OrientVectorCount=::NewEnumerator1
      Side Direction:Vector3f=?
      Up Direction:Vector3f=<0 0 1>
      Angle Type=::NewEnumerator0
      Coordinate Space=::Local
      Normalize:NiagaraBool=true
      Rotation Matrix=Linked() {linkedVariable {name=Engine.Owner.SystemLocalToWorldNoScale, type=NiagaraFloat}}
     )
    )
    Sampled Position Offset Coordinate Space:ENiagaraSkelMeshTransforms=::NewEnumerator0
    Skeletal Mesh=$System.Skeletal Mesh
   )
   &OffsetPosition(Position Offset=&Multiply_VectorByFloat(
     Vector=$System.DomeDirecrion
     Float=&Sine(
      Normalized Angle=&Add_Float(A=$Emitter.Age, B=&RandomRangeFloat(Minimum=0, Maximum=3.14))
      Period=&RandomRangeFloat(Minimum=1, Maximum=2)
      Scale=&Lerp_Float(
       A=&RandomRangeFloat(Minimum=14, Maximum=18)
       B=&RandomRangeFloat(Minimum=5, Maximum=10)
       Alpha=$User.ParaGlideLeanAlpha
      )
      Bias=0
     )
    ))
   &UpdateMeshOrientation(
    Facing Direction=&NormalizeVector(
     Vector To Normalize=&ConvertPositionToVector(Input Position=$Particles.Position, Conversion Operation:ENiagara_LWCConvertPosToVec=::NewEnumerator0)
     Safe Normalize=true
     Safe Fallback Threshold=0
     Safe Fallback Vector=<0 0 0.001>
    )
    Facing Interpolation=false
    Orientation Method=::NewEnumerator1
    Override Quat=false
    TransformFacing=false
    TransformUp=false
    Up Direction:Vector3f=$System.DomeDirecrion
   )
   &ScaleColor(Scale Alpha=$System.Dive, ScaleA=true, ScaleRGB=true) {enabled=false}
   /SetVariables(
    Particles.Frame_Color:LinearColor=<rgba 1 0.5448 0.0604 0.25>
    Particles.Frame_DynGlow:Vector4f=&MakeVector4(X=0.3, Y=1, Z=1, W=4)
    Particles.Frame_Scale:Vector3f=&MakeVector(
     X=0.001
     Y=&MakeFloatFromVector(Vector=$Particles.Scale, Channel=::NewEnumerator1)
     Z=&MakeFloatFromVector(Vector=$Particles.Scale, Channel=::NewEnumerator2)
    )
   ) {isSetParameters=true, module=SetVariables_#11}
   &AlignSpriteToMeshOrientation(Mesh Orientation Relative Sprite Alignment Vector=<0 0 1>, SetAlignment=true, SetFacing=true)
   /SetVariables(Particles.SpritePosition:NiagaraPosition=&AddVectorToPosition(Position=$Particles.Position, Vector=<0 0 -24.0527>)) {isSetParameters=true}
   &FN_Mesh_Attributes(
    Camera Offset=-100
    Color=&MakeLinearColorFromVectorAndFloat(
     "Vector (RGB)"=&MakeVectorFromLinearColorRGB(Color=<rgba 40 20.0795 0 0.113>)
     "Float (Alpha)"=&Multiply_Float(
      A=&MakeCustomFloatFromBool(
       Bool=$User.bIsFrontEndPreview
       True Float=&Lerp_Float(A=0.15, B=0.3, Alpha=$System.FN_Contrail_Setup.Dive Value)
       False Float=&Multiply_Float(A=&Lerp_Float(A=0.5, B=1, Alpha=$System.FN_Contrail_Setup.Dive Value), B=0.8)
      )
      B=0.2
     )
    )
    Create Camera Offset Attribute=true
    Create Color Attribute=true
    Create Scale Attribute=true
    Module Output=::NewEnumerator1
    Scale=&Multiply_VectorByFloat(
     Vector=&Multiply_Vector(
      A=&Multiply_VectorByFloat(
       Vector=&Lerp_Vector(
        A=&RandomRangeVector(Minimum=<1 0.5 0.45>, Maximum=<1 0.9 0.52>)
        B=&RandomRangeVector(Minimum=<1 0.5 0.55>, Maximum=<1 0.9 0.62>)
        Alpha=$System.SmoothDiveAlpha
       )
       Float=&RandomRangeFloat(Minimum=1, Maximum=1)
      )
      B=&MakeCustomVectorFromBool(
       Bool=$User.bIsFrontEndPreview
       True Vector=&Multiply_Vector(
        A=&Lerp_Vector(
         A=<0.4 0.4 0.45>
         B=<0.85 0.85 0.8>
         Alpha=&ClampFloat(Float=&Add_Float(A=$System.FN_Contrail_Setup.Dive Value, B=1))
        )
        B=<1.5 1.5 3>
       )
       False Vector=&Lerp_Vector(
        A=<0.4 0.4 0.45>
        B=<0.85 0.85 0.8>
        Alpha=&ClampFloat(Float=&Add_Float(A=$System.FN_Contrail_Setup.Dive Value, B=1))
       )
      )
     )
     Float=&Multiply_Float(
      A=1
      B=&FloatFromCurve(
       FloatCurve=DataInterface() {propertyValues {Curve {
          defaultValue=0
          keys [{interpMode=RCIM_Cubic, tangentMode=RCTM_User, time=0, value=0}, {interpMode=RCIM_Cubic, tangentMode=RCTM_User, time=1, value=1}]
         }}}
       CurveIndex=$Engine.System.Age
      )
     )
    )
   ) {module=FN_Mesh_Attributes001}
   &FN_Mesh_Attributes(
    Color=&MakeLinearColorFromVectorAndFloat(
     "Vector (RGB)"=&MakeVectorFromLinearColorRGB(Color=<rgba 25.9055 0 40 1>)
     "Float (Alpha)"=&Multiply_Float(
      A=&MakeCustomFloatFromBool(
       Bool=$User.bIsFrontEndPreview
       True Float=&Lerp_Float(A=0.15, B=0.3, Alpha=$System.FN_Contrail_Setup.Dive Value)
       False Float=&Multiply_Float(A=&Lerp_Float(A=0.5, B=1, Alpha=$System.FN_Contrail_Setup.Dive Value), B=0.8)
      )
      B=0.2
     )
    )
    Create Color Attribute=true
    Create Scale Attribute=true
    Module Output=::NewEnumerator1
    Scale=&Multiply_VectorByFloat(
     Vector=&Multiply_Vector(
      A=&Multiply_VectorByFloat(
       Vector=&Lerp_Vector(
        A=&RandomRangeVector(Minimum=<1 1.2 0.58>, Maximum=<1 1.8 0.68>)
        B=&RandomRangeVector(Minimum=<1 1.2 0.78>, Maximum=<1 1.8 0.88>)
        Alpha=$System.SmoothDiveAlpha
       )
       Float=&RandomRangeFloat(Minimum=1, Maximum=1)
      )
      B=&MakeCustomVectorFromBool(
       Bool=$User.bIsFrontEndPreview
       True Vector=&Multiply_Vector(
        A=&Lerp_Vector(
         A=<0.4 0.4 0.45>
         B=<0.85 0.85 0.8>
         Alpha=&ClampFloat(Float=&Add_Float(A=$System.FN_Contrail_Setup.Dive Value, B=1))
        )
        B=<1.5 1.5 3>
       )
       False Vector=&Lerp_Vector(
        A=<0.4 0.4 0.45>
        B=<0.85 0.85 0.8>
        Alpha=&ClampFloat(Float=&Add_Float(A=$System.FN_Contrail_Setup.Dive Value, B=1))
       )
      )
     )
     Float=&Multiply_Float(
      A=1
      B=&FloatFromCurve(
       FloatCurve=DataInterface() {propertyValues {Curve {
          defaultValue=0
          keys [{interpMode=RCIM_Cubic, tangentMode=RCTM_User, time=0, value=0}, {interpMode=RCIM_Cubic, tangentMode=RCTM_User, time=1, value=1}]
         }}}
       CurveIndex=$Engine.System.Age
      )
     )
    )
   )
   &FN_Mesh_Attributes(
    Color=&MakeLinearColorFromVectorAndFloat(
     "Vector (RGB)"=&MakeVectorFromLinearColorRGB(Color=<rgba 25.9055 0 40 1>)
     "Float (Alpha)"=&Multiply_Float(
      A=&MakeCustomFloatFromBool(
       Bool=$User.bIsFrontEndPreview
       True Float=&Lerp_Float(A=0.15, B=0.3, Alpha=$System.FN_Contrail_Setup.Dive Value)
       False Float=&Multiply_Float(A=&Lerp_Float(A=0.5, B=1, Alpha=$System.FN_Contrail_Setup.Dive Value), B=0.8)
      )
      B=0.2
     )
    )
    Create Color Attribute=true
    Create Scale Attribute=true
    Module Output=::NewEnumerator1
    Scale=&Multiply_VectorByFloat(
     Vector=&Multiply_Vector(
      A=&Multiply_VectorByFloat(
       Vector=&RandomRangeVector(Minimum=<1 0.8 1>, Maximum=<1 0.9 1.2>)
       Float=&RandomRangeFloat(Minimum=1, Maximum=1)
      )
      B=&MakeCustomVectorFromBool(
       Bool=$User.bIsFrontEndPreview
       True Vector=&Multiply_Vector(
        A=&Lerp_Vector(
         A=<0.4 0.4 0.45>
         B=<0.85 0.85 0.8>
         Alpha=&ClampFloat(Float=&Add_Float(A=$System.FN_Contrail_Setup.Dive Value, B=1))
        )
        B=<1.5 1.5 3>
       )
       False Vector=&Lerp_Vector(
        A=<0.4 0.4 0.45>
        B=<0.85 0.85 0.8>
        Alpha=&ClampFloat(Float=&Add_Float(A=$System.FN_Contrail_Setup.Dive Value, B=0.55))
       )
      )
     )
     Float=&Multiply_Float(
      A=1
      B=&FloatFromCurve(
       FloatCurve=DataInterface() {propertyValues {Curve {
          defaultValue=0
          keys [{interpMode=RCIM_Cubic, tangentMode=RCTM_User, time=0, value=0}, {interpMode=RCIM_Cubic, tangentMode=RCTM_User, time=1, value=1}]
         }}}
       CurveIndex=$Engine.System.Age
      )
     )
    )
   ) {module=FN_Mesh_Attributes002}
  ]
  rendererClasses [&NiagaraMeshRendererProperties, &NiagaraSpriteRendererProperties]
  renderers [
   {rendererClass=&NiagaraMeshRendererProperties, rendererIndex=0}
   {rendererClass=&NiagaraMeshRendererProperties, rendererIndex=1}
   {rendererClass=&NiagaraSpriteRendererProperties, rendererIndex=2}
   {rendererClass=&NiagaraMeshRendererProperties, rendererIndex=3}
   {rendererClass=&NiagaraMeshRendererProperties, rendererIndex=4}
   {rendererClass=&NiagaraMeshRendererProperties, rendererIndex=5}
  ]
 }
 Card_Dome001 {
  emitterUpdateScript [
   &EmitterState(Life Cycle Mode=::NewEnumerator1, Loop Behavior=::NewEnumerator1, Loop Duration Mode=::NewEnumerator1, UseLoopDelay=false)
   &SpawnBurst_Instantaneous(Spawn Count=9, Use Spawn Probability=false, UseLoopCountLimit=false)
  ]
  particleSpawnScript [
   &InitializeParticle(
    Color Mode=::NewEnumerator0
    Lifetime=5
    Mesh Scale=&MakeVector(X=2, Y=2, Z=&RandomRangeFloat(Minimum=2, Maximum=4))
    Mesh Scale Mode=::NewEnumerator4
    Sprite Size Mode=::NewEnumerator1
    Uniform Sprite Size Max=100
    Uniform Sprite Size Min=50
    UsePositionOffset=false
    Write Material Random=true
   )
   &SystemLocation(Offset=&RotateVectorByQuaternion(
     Vector To Rotate=&RotateVector(
      Vector To Rotate:Vector3f=&MakeVector(X=$System.DomeRadius, Y=0, Z=0)
      Yaw:NiagaraFloat=&ReturnNormalizedExecIndex(Normalized Index Scale=720, Normalized Index Includes One=false)
      Pitch:NiagaraFloat=&RandomRangeFloat(Minimum=-45, Maximum=0)
      Roll:NiagaraFloat=0
      Rotation Rate:NiagaraFloat=1
     )
     Quaternion=&MakeQuaternion(
      Quaternion From=3
      Axis=<1 0 0>
      Angle=0
      From=<0 0 1>
      To=$System.DomeDirecrion
      Stabilization Mode:ENiagara_OrientVectorCount=::NewEnumerator1
      Side Direction:Vector3f=?
      Up Direction:Vector3f=<0 0 1>
      Angle Type=::NewEnumerator0
      Coordinate Space=::Local
      Normalize:NiagaraBool=true
      Rotation Matrix=Linked() {linkedVariable {name=Engine.Owner.SystemLocalToWorldNoScale, type=NiagaraFloat}}
     )
    )) {enabled=false}
   /SetVariables(Particles.Position=&AddVectorToPosition(Position=$Particles.Position, Vector=<0 0 110>)) {enabled=false, isSetParameters=true, module=SetVariables_#9}
   /SetVariables(Particles.CardIndex:NiagaraFloat=&RandomRangeFloat(Minimum=0, Maximum=3)) {enabled=false, isSetParameters=true, module=SetVariables_#10}
  ]
  particleUpdateScript [
   &ParticleState(Kill Particles When Lifetime Has Elapsed=false, Let Infinitely Lived Particles Die When Emitter Deactivates=true)
   &SkeletalMeshLocation(
    Bone / Socket Index=4
    Bone Sampling Mode=::NewEnumerator4
    MaskSpawnGroup=false
    Orientation Sampling=::NewEnumerator1
    Sampled Position Offset=&RotateVectorByQuaternion(
     Vector To Rotate=&RotateVector(
      Vector To Rotate:Vector3f=&MakeVector(X=$System.DomeRadius, Y=0, Z=0)
      Yaw:NiagaraFloat=&ReturnNormalizedExecIndex(Normalized Index Scale=720, Normalized Index Includes One=false)
      Pitch:NiagaraFloat=&RandomRangeFloat(Minimum=-15, Maximum=-30)
      Roll:NiagaraFloat=0
      Rotation Rate:NiagaraFloat=1
     )
     Quaternion=&MakeQuaternion(
      Quaternion From=3
      Axis=<1 0 0>
      Angle=0
      From=<0 0 1>
      To=/Niagara/DynamicInputs/Vectors/NegateVector(Vector:Vector3f=$System.DomeDirecrion)
      Stabilization Mode:ENiagara_OrientVectorCount=::NewEnumerator0
      Side Direction:Vector3f=?
      Up Direction:Vector3f=?
      Angle Type=::NewEnumerator0
      Normalize:NiagaraBool=true
      Rotation Matrix=Linked() {linkedVariable {name=Engine.Owner.SystemLocalToWorldNoScale, type=NiagaraFloat}}
     )
    )
    Sampled Position Offset Coordinate Space:ENiagaraSkelMeshTransforms=::NewEnumerator0
    Skeletal Mesh=$System.Skeletal Mesh
   )
   &OffsetPosition(Position Offset=&Multiply_VectorByFloat(
     Vector=$System.DomeDirecrion
     Float=&Sine(
      Normalized Angle=&Add_Float(A=$Emitter.Age, B=&RandomRangeFloat(Minimum=0, Maximum=1))
      Period=&RandomRangeFloat(Minimum=1, Maximum=2)
      Scale=&Lerp_Float(
       A=&RandomRangeFloat(Minimum=4, Maximum=8)
       B=&RandomRangeFloat(Minimum=8, Maximum=12)
       Alpha=$User.ParaGlideLeanAlpha
      )
      Bias=0
     )
    ))
   &UpdateMeshOrientation(
    Facing Direction=&NormalizeVector(
     Vector To Normalize=&Subtract_Position(A=$Particles.Position, B=$Output.SkeletalMeshLocation.SampledPositionWithoutOffset)
     Safe Normalize=true
     Safe Fallback Threshold=0
     Safe Fallback Vector=<0 0 0.001>
    )
    Facing Interpolation=false
    Orientation Method=::NewEnumerator1
    Override Quat=false
    TransformFacing=false
    TransformUp=false
    Up Direction:Vector3f=$System.DomeDirecrion
   )
   /SetVariables(
    Particles.Frame_Color:LinearColor=<rgba 2 0.5333 0 0.25>
    Particles.Frame_DynGlow:Vector4f=&MakeVector4(X=0.3, Y=1, Z=1, W=4)
    Particles.Frame_Scale:Vector3f=&MakeVector(
     X=0.001
     Y=&MakeFloatFromVector(Vector=$Particles.Scale, Channel=::NewEnumerator1)
     Z=&MakeFloatFromVector(Vector=$Particles.Scale, Channel=::NewEnumerator2)
    )
   ) {isSetParameters=true, module=SetVariables_#11}
   &FN_Mesh_Attributes(
    Camera Offset=-100
    Color=&MakeLinearColorFromVectorAndFloat(
     "Vector (RGB)"=&MakeVectorFromLinearColorRGB(Color=<rgba 60 16.8775 3.7996 1>)
     "Float (Alpha)"=&Multiply_Float(
      A=&MakeCustomFloatFromBool(
       Bool=$User.bIsFrontEndPreview
       True Float=&Lerp_Float(A=0.15, B=0.3, Alpha=$System.FN_Contrail_Setup.Dive Value)
       False Float=&Multiply_Float(A=&Lerp_Float(A=0.5, B=1, Alpha=$System.FN_Contrail_Setup.Dive Value), B=0.8)
      )
      B=0.2
     )
    )
    Create Camera Offset Attribute=true
    Create Color Attribute=true
    Create Scale Attribute=true
    Module Output=::NewEnumerator1
    Scale=&Multiply_VectorByFloat(
     Vector=&Multiply_Vector(
      A=&Multiply_VectorByFloat(
       Vector=&Lerp_Vector(
        A=&RandomRangeVector(Minimum=<1 0.4 0.5>, Maximum=<1 0.6 0.62>)
        B=&RandomRangeVector(Minimum=<1 0.8 0.65>, Maximum=<1 1 0.7>)
        Alpha=$System.SmoothDiveAlpha
       )
       Float=&RandomRangeFloat(Minimum=1, Maximum=1)
      )
      B=&MakeCustomVectorFromBool(
       Bool=$User.bIsFrontEndPreview
       True Vector=&Multiply_Vector(
        A=&Lerp_Vector(
         A=<0.4 0.4 0.45>
         B=<0.85 0.85 0.8>
         Alpha=&ClampFloat(Float=&Add_Float(A=$System.FN_Contrail_Setup.Dive Value, B=1))
        )
        B=<1.5 1.5 3>
       )
       False Vector=&Lerp_Vector(
        A=<0.4 0.4 0.45>
        B=<0.85 0.85 0.8>
        Alpha=&ClampFloat(Float=&Add_Float(A=$System.FN_Contrail_Setup.Dive Value, B=1))
       )
      )
     )
     Float=&Multiply_Float(
      A=1
      B=&FloatFromCurve(
       FloatCurve=DataInterface() {propertyValues {Curve {
          defaultValue=0
          keys [{interpMode=RCIM_Cubic, tangentMode=RCTM_User, time=0, value=0}, {interpMode=RCIM_Cubic, tangentMode=RCTM_User, time=1, value=1}]
         }}}
       CurveIndex=$Engine.System.Age
      )
     )
    )
   ) {module=FN_Mesh_Attributes001}
   &FN_Mesh_Attributes(
    Color=&MakeLinearColorFromVectorAndFloat(
     "Vector (RGB)"=&MakeVectorFromLinearColorRGB(Color=<rgba 16.8234 0 25 1>)
     "Float (Alpha)"=&Multiply_Float(
      A=&MakeCustomFloatFromBool(
       Bool=$User.bIsFrontEndPreview
       True Float=&Lerp_Float(A=0.15, B=0.3, Alpha=$System.FN_Contrail_Setup.Dive Value)
       False Float=&Multiply_Float(A=&Lerp_Float(A=0.5, B=1, Alpha=$System.FN_Contrail_Setup.Dive Value), B=0.8)
      )
      B=0.2
     )
    )
    Create Color Attribute=true
    Create Scale Attribute=true
    Module Output=::NewEnumerator1
    Scale=&Multiply_VectorByFloat(
     Vector=&Multiply_Vector(
      A=&Multiply_VectorByFloat(
       Vector=&Lerp_Vector(
        A=&RandomRangeVector(Minimum=<1 2 0.65>, Maximum=<1 3 0.75>)
        B=&RandomRangeVector(Minimum=<1 2.5 0.85>, Maximum=<2 3.5 0.95>)
        Alpha=$System.SmoothDiveAlpha
       )
       Float=&RandomRangeFloat(Minimum=1, Maximum=1)
      )
      B=&MakeCustomVectorFromBool(
       Bool=$User.bIsFrontEndPreview
       True Vector=&Multiply_Vector(
        A=&Lerp_Vector(
         A=<0.4 0.4 0.45>
         B=<0.85 0.85 0.8>
         Alpha=&ClampFloat(Float=&Add_Float(A=$System.FN_Contrail_Setup.Dive Value, B=1))
        )
        B=<1.5 1.5 3>
       )
       False Vector=&Lerp_Vector(
        A=<0.4 0.4 0.45>
        B=<0.85 0.85 0.8>
        Alpha=&ClampFloat(Float=&Add_Float(A=$System.FN_Contrail_Setup.Dive Value, B=1))
       )
      )
     )
     Float=&Multiply_Float(
      A=1
      B=&FloatFromCurve(
       FloatCurve=DataInterface() {propertyValues {Curve {
          defaultValue=0
          keys [{interpMode=RCIM_Cubic, tangentMode=RCTM_User, time=0, value=0}, {interpMode=RCIM_Cubic, tangentMode=RCTM_User, time=1, value=1}]
         }}}
       CurveIndex=$Engine.System.Age
      )
     )
    )
   )
   &FN_Mesh_Attributes(
    Color=&MakeLinearColorFromVectorAndFloat(
     "Vector (RGB)"=&MakeVectorFromLinearColorRGB(Color=<rgba 25.9055 0 40 1>)
     "Float (Alpha)"=&Multiply_Float(
      A=&MakeCustomFloatFromBool(
       Bool=$User.bIsFrontEndPreview
       True Float=&Lerp_Float(A=0.15, B=0.3, Alpha=$System.FN_Contrail_Setup.Dive Value)
       False Float=&Multiply_Float(A=&Lerp_Float(A=0.5, B=1, Alpha=$System.FN_Contrail_Setup.Dive Value), B=0.8)
      )
      B=0.2
     )
    )
    Create Color Attribute=true
    Create Scale Attribute=true
    Module Output=::NewEnumerator1
    Scale=&Multiply_VectorByFloat(
     Vector=&Multiply_Vector(
      A=&Multiply_VectorByFloat(
       Vector=&RandomRangeVector(Minimum=<1 0.8 1>, Maximum=<1 0.9 1.2>)
       Float=&RandomRangeFloat(Minimum=1, Maximum=1)
      )
      B=&MakeCustomVectorFromBool(
       Bool=$User.bIsFrontEndPreview
       True Vector=&Multiply_Vector(
        A=&Lerp_Vector(
         A=<0.4 0.4 0.45>
         B=<0.85 0.85 0.8>
         Alpha=&ClampFloat(Float=&Add_Float(A=$System.FN_Contrail_Setup.Dive Value, B=1))
        )
        B=<1.5 1.5 3>
       )
       False Vector=&Lerp_Vector(
        A=<0.4 0.4 0.45>
        B=<0.85 0.85 0.8>
        Alpha=&ClampFloat(Float=&Add_Float(A=$System.FN_Contrail_Setup.Dive Value, B=0.55))
       )
      )
     )
     Float=&Multiply_Float(
      A=1
      B=&FloatFromCurve(
       FloatCurve=DataInterface() {propertyValues {Curve {
          defaultValue=0
          keys [{interpMode=RCIM_Cubic, tangentMode=RCTM_User, time=0, value=0}, {interpMode=RCIM_Cubic, tangentMode=RCTM_User, time=1, value=1}]
         }}}
       CurveIndex=$Engine.System.Age
      )
     )
    )
   ) {module=FN_Mesh_Attributes002}
   /SetVariables(Particles.ResetRot:Quat4f=&MakeQuaternion(
     Quaternion From=3
     Axis=<1 0 0>
     Angle=0
     From=<0 0 1>
     To=&NormalizeVector(
      Vector To Normalize=$System.FN_Contrail_Setup.Direction Velocity
      Safe Normalize=true
      Safe Fallback Threshold=0
      Safe Fallback Vector=<0 0 0.001>
     )
     Stabilization Mode:ENiagara_OrientVectorCount=::NewEnumerator0
     Side Direction:Vector3f=?
     Up Direction:Vector3f=?
     Angle Type=::NewEnumerator0
     Coordinate Space=::Local
     Normalize:NiagaraBool=true
    )) {isSetParameters=true}
   /SetVariables(Particles.OffsetPos:NiagaraPosition=&AddVectorToPosition(
     Position=$Particles.Position
     Vector=&RotateVectorByQuaternion(Vector To Rotate=<3 0 13>, Quaternion=$Particles.MeshOrientation)
    )) {isSetParameters=true}
  ]
  rendererClasses [&NiagaraMeshRendererProperties]
  renderers [
   {rendererClass=&NiagaraMeshRendererProperties, rendererIndex=0}
   {rendererClass=&NiagaraMeshRendererProperties, rendererIndex=1}
   {rendererClass=&NiagaraMeshRendererProperties, rendererIndex=2}
   {rendererClass=&NiagaraMeshRendererProperties, rendererIndex=3}
   {rendererClass=&NiagaraMeshRendererProperties, rendererIndex=4}
  ]
 }
 Card_Stream {
  bEnabled=false
  emitterUpdateScript [
   &EmitterState(Life Cycle Mode=::NewEnumerator1, UseLoopDelay=false)
   &SpawnRate(SpawnRate=&RandomRangeFloat(Minimum=5, Maximum=10), Use Spawn Probability=false)
  ]
  particleSpawnScript [
   &InitializeParticle(
    Color Mode=::NewEnumerator0
    Lifetime Max=1.5
    Lifetime Mode=::NewEnumerator1
    Mesh Scale Mode=::NewEnumerator1
    Mesh Uniform Scale Max=2.2
    Mesh Uniform Scale Min=1.2
    UsePositionOffset=false
    Write Material Random=true
   )
   &SystemLocation(Offset=&Multiply_VectorByFloat(Vector=$System.DomeDirecrion, Float=24))
   &AddVelocity(
    Velocity=&RotateVectorByQuaternion(
     Vector To Rotate=&RotateVector(
      Vector To Rotate:Vector3f=&NormalizeVector(Vector To Normalize=<1 0 -1>, Safe Normalize=true, Safe Fallback Threshold=0, Safe Fallback Vector=<0 0 0.001>)
      Yaw:NiagaraFloat=&RandomRangeFloat(Minimum=0, Maximum=360)
      Pitch:NiagaraFloat=0
      Roll:NiagaraFloat=0
      Rotation Rate:NiagaraFloat=1
     )
     Quaternion=&MakeQuaternion(
      Quaternion From=3
      Axis=<1 0 0>
      Angle=0
      From=<0 0 -1>
      To=$System.DomeDirecrion
      Stabilization Mode:ENiagara_OrientVectorCount=::NewEnumerator0
      Side Direction:Vector3f=?
      Up Direction:Vector3f=?
      Angle Type=::NewEnumerator0
      Normalize:NiagaraBool=true
      Rotation Matrix=Linked() {linkedVariable {name=Engine.Owner.SystemLocalToWorldNoScale, type=NiagaraFloat}}
     )
    )
    Velocity Speed Scale=&RandomRangeFloat(Minimum=200, Maximum=400)
   )
   &InitialMeshOrientation(Use Rotation Vector=false)
   &SystemLocation(Offset=<0 0 20>) {enabled=false, module=SystemLocation001}
   &ShapeLocation(Handle Radius=30, Large Radius=60, MaskSpawnGroup=false, Override Seed=false, Shape Primitive=::NewEnumerator3, UseOwnerScale=true)
  ]
  particleUpdateScript [
   &ParticleState()
   &SkeletalMeshLocation(
    Bone / Socket Index=4
    Bone Sampling Mode=::NewEnumerator4
    MaskSpawnGroup=false
    Orientation Sampling=::NewEnumerator1
    Position Sampling=::NewEnumerator1
    Sampled Position Offset=&Lerp_Vector(A=<0 0 -72>, B=<0 0 -60>, Alpha=$User.ParaGlideLeanAlpha)
    Sampled Position Offset Coordinate Space:ENiagaraSkelMeshTransforms=::NewEnumerator0
    Skeletal Mesh=$System.Skeletal Mesh
   ) {enabled=false}
   &OffsetPosition(Position Offset=&Lerp_Vector(A=<0 0 0>, B=<0 0 0>, Alpha=$User.ParaGlideLeanAlpha)) {enabled=false}
   &ScaleMeshSize(Scale Factor=&VectorFromFloat(Value=&FloatFromCurve(FloatCurve=DataInterface() {propertyValues {Curve {
         defaultValue=1
         keys [
          {tangentMode=RCTM_User, time=0, value=1}
          {arriveTangent=-0.0405, interpMode=RCIM_Cubic, leaveTangent=-0.0405, tangentMode=RCTM_User, time=0.7009, value=1}
          {arriveTangent=-8.0089, interpMode=RCIM_Cubic, leaveTangent=-8.0089, tangentMode=RCTM_User, time=1, value=0}
         ]
        }}})))
   &AerodynamicDrag(
    Aerodynamic Drag=0.05
    Rotation Strength:NiagaraFloat=&RandomRangeFloat(Minimum=10, Maximum=30)
    Use Linear Drag=true
    Use Rotational Drag=true
   )
   &ScaleVelocity(Velocity Scale=&Lerp_Vector(A=<1 1 1>, B=<2.2 2.2 1.3>, Alpha=$User.ParaGlideLeanAlpha))
   &SolveForcesAndVelocity(Clamp Velocity=false, Limit Acceleration=false)
   &AlignSpriteToMeshOrientation(SetAlignment=true, SetFacing=true)
   &KillParticlesInVolume(Box Size:Vector3f=<60 0 20>, Kill Shape=::NewEnumerator1, Origin Offset:Vector3f=<0 -100 0>) {module=KillParticlesInVolume001}
  ]
  rendererClasses [&NiagaraMeshRendererProperties]
  renderers [{rendererClass=&NiagaraMeshRendererProperties, rendererIndex=0}]
 }
 Card_Stream001 {
  emitterUpdateScript [
   &EmitterState(Life Cycle Mode=::NewEnumerator1, UseLoopDelay=false)
   &SpawnRate(SpawnRate=&Lerp_Float(A=3, B=12, Alpha=$User.ParaGlideLeanAlpha), Use Spawn Probability=false)
  ]
  particleSpawnScript [
   &InitializeParticle(
    Color Mode=::NewEnumerator0
    Lifetime Max=1.5
    Lifetime Mode=::NewEnumerator1
    Mesh Scale Mode=::NewEnumerator1
    Mesh Uniform Scale Max=4
    Mesh Uniform Scale Min=3
    UsePositionOffset=false
    Write Material Random=true
   )
   &SystemLocation(Offset=&Multiply_VectorByFloat(Vector=$System.DomeDirecrion, Float=24))
   &AddVelocity(
    Velocity=&RotateVectorByQuaternion(
     Vector To Rotate=&RotateVector(
      Vector To Rotate:Vector3f=&NormalizeVector(Vector To Normalize=<1 0 -1>, Safe Normalize=true, Safe Fallback Threshold=0, Safe Fallback Vector=<0 0 0.001>)
      Yaw:NiagaraFloat=&RandomRangeFloat(Minimum=0, Maximum=360)
      Pitch:NiagaraFloat=0
      Roll:NiagaraFloat=0
      Rotation Rate:NiagaraFloat=1
     )
     Quaternion=&MakeQuaternion(
      Quaternion From=3
      Axis=<1 0 0>
      Angle=0
      From=<0 0 -1>
      To=$System.DomeDirecrion
      Stabilization Mode:ENiagara_OrientVectorCount=::NewEnumerator0
      Side Direction:Vector3f=?
      Up Direction:Vector3f=?
      Angle Type=::NewEnumerator0
      Normalize:NiagaraBool=true
      Rotation Matrix=Linked() {linkedVariable {name=Engine.Owner.SystemLocalToWorldNoScale, type=NiagaraFloat}}
     )
    )
    Velocity Speed Scale=&RandomRangeFloat(Minimum=200, Maximum=400)
   )
   &InitialMeshOrientation(Use Rotation Vector=false)
   &ShapeLocation(Handle Radius=30, Large Radius=60, MaskSpawnGroup=false, Override Seed=false, Shape Primitive=::NewEnumerator3, UseOwnerScale=true)
  ]
  particleUpdateScript [
   &ParticleState()
   &SkeletalMeshLocation(
    Bone / Socket Index=4
    Bone Sampling Mode=::NewEnumerator4
    MaskSpawnGroup=false
    Orientation Sampling=::NewEnumerator1
    Position Sampling=::NewEnumerator1
    Sampled Position Offset=&Lerp_Vector(A=<0 0 -72>, B=<0 0 -60>, Alpha=$User.ParaGlideLeanAlpha)
    Sampled Position Offset Coordinate Space:ENiagaraSkelMeshTransforms=::NewEnumerator0
    Skeletal Mesh=$System.Skeletal Mesh
   ) {enabled=false}
   &OffsetPosition(Position Offset=&Lerp_Vector(A=<0 0 0>, B=<0 0 0>, Alpha=$User.ParaGlideLeanAlpha)) {enabled=false}
   &ScaleMeshSize(Scale Factor=&VectorFromFloat(Value=&FloatFromCurve(FloatCurve=DataInterface() {propertyValues {Curve {
         defaultValue=1
         keys [
          {tangentMode=RCTM_User, time=0, value=1}
          {arriveTangent=-0.0405, interpMode=RCIM_Cubic, leaveTangent=-0.0405, tangentMode=RCTM_User, time=0.7009, value=1}
          {arriveTangent=-8.0089, interpMode=RCIM_Cubic, leaveTangent=-8.0089, tangentMode=RCTM_User, time=1, value=0}
         ]
        }}})))
   &AerodynamicDrag(
    Aerodynamic Drag=0.05
    Rotation Strength:NiagaraFloat=&RandomRangeFloat(Minimum=10, Maximum=30)
    Use Linear Drag=true
    Use Rotational Drag=true
   )
   &ScaleVelocity(Velocity Scale=&Lerp_Vector(A=<1 1 1>, B=<2.2 2.2 1.3>, Alpha=$User.ParaGlideLeanAlpha))
   &SolveForcesAndVelocity(Clamp Velocity=false, Limit Acceleration=false)
   &AlignSpriteToMeshOrientation(SetAlignment=true, SetFacing=true)
   &KillParticlesInVolume(Box Size:Vector3f=<60 0 20>, Kill Shape=::NewEnumerator1, Origin Offset:Vector3f=<0 -100 0>) {module=KillParticlesInVolume001}
  ]
  rendererClasses [&NiagaraMeshRendererProperties]
  renderers [{rendererClass=&NiagaraMeshRendererProperties, rendererIndex=0}]
 }
 Center001 {
  bEnabled=false
  emitterSpawnScript [&FN_PreserveSkydiveUserVariables() {module=PreserveSkydiveUserVariables}]
  emitterUpdateScript [
   &EmitterState(
    Life Cycle Mode=::NewEnumerator1
    Loop Behavior=::NewEnumerator1
    Loop Duration Mode=::NewEnumerator1
    Reset Age On Awaken=true
    Scalability Mode=::NewEnumerator1
    UseLoopDelay=false
   )
   &SpawnBurst_Instantaneous(Spawn Count=1, Use Spawn Probability=false)
  ]
  particleSpawnScript [
   &FN_SetSpawnGroup(Number of Spawn Groups=1) {module=FN_SetSpawnGroup001}
   &InitializeParticle(
    Color Mode=::NewEnumerator0
    Mesh Scale=<1.3 1.3 0.6>
    Mesh Scale Mode=::NewEnumerator4
    Sprite Size Mode=::NewEnumerator3
    Uniform Sprite Size=400
    UsePositionOffset=false
    Write Material Random=true
   )
   &InitialMeshOrientation(
    Orientation Axis=<0 0 1>
    Orientation Coordinate Space=::Local
    Orientation Vector=<0 0 1>
    Use Orientation Vector=true
    Use Rotation Vector=false
   ) {enabled=false}
   &InitialMeshOrientation(
    Mesh Orientation Mode=::NewEnumerator2
    Orientation Coordinate Space=::Local
    Orientation Vector=<0 0 1>
    Rotation=<1 0 0>
    Rotation Coordinate Space=::NewEnumerator2
    Use Rotation Vector=false
   ) {module=InitialMeshOrientation002}
   &DynamicMaterialParameters(
    Index 0 Param 1=3
    Index 0 Param 2=0.4
    Module.Index0Float0Write=true
    Module.Index0Float1Write=true
    Module.Index0Float2Write=true
    Module.Index0Float3Write=true
    Module.Index2Float0Write=true
    Module.Index2Float1Write=true
    Module.Index2Float2Write=true
    Module.Index2Float3Write=true
    Module.Index3Float0Write=true
    Module.Index3Float1Write=true
    Module.Index3Float2Write=true
    Module.Index3Float3Write=true
    Write Parameter Index 2=::NewEnumerator1
    Write Parameter Index 3=::NewEnumerator1
   ) {enabled=false}
  ]
  particleUpdateScript [
   /SetVariables(Particles.AlphaAll:NiagaraFloat=&Multiply_Float(A=0.8, B=&Lerp_Float(A=0, B=1, Alpha=1))) {isSetParameters=true, module=SetVariables_#12}
   /Game/Effects/Niagara/Modules/Skydiving/FN_DivingAttributes/FN_DivingAttributes_v2(Accumulate Dive=0.1, Accumulate Rest=0.1, ParaGlideLeanAlpha:NiagaraFloat=$User.ParaGlideLeanAlpha) {enabled=false, module=FN_DivingAttributes_v2001}
   &ParticleState(Kill Particles When Lifetime Has Elapsed=false, Let Infinitely Lived Particles Die When Emitter Deactivates=true)
   &ScaleMeshSize(Scale Factor=&Multiply_VectorByFloat(
     Vector=&Multiply_VectorByFloat(Vector=<1 1 1>, Float=&FN_FrontEndFloatModifier(Gameplay=1, Front End=0.9))
     Float=&Lerp_Float(A=1, B=1.3, Alpha=&ClampFloat(Float=$User.ParaGlideLeanAlpha))
    )) {module=ScaleMeshSize002}
   &SkeletalMeshLocation(
    Bone / Socket Index=4
    Bone Sampling Mode=::NewEnumerator4
    MaskSpawnGroup=false
    Orientation Sampling=::NewEnumerator1
    Sampled Position Offset=&Lerp_Vector(A=<0 65 0>, B=<0 35 0>, Alpha=$User.ParaGlideLeanAlpha)
    Skeletal Mesh=$System.Skeletal Mesh
   ) {module=SkeletalMeshLocation001}
   &UpdateMeshOrientation(
    Facing Direction=&Multiply_Vector(A=Linked() {linkedVariable {name=Particles.SkeletalMeshLocation001.SampledPosition, type=NiagaraPosition}}, B=<-1 1 1>)
    Facing Interpolation=true
    Facing Interpolation Rate:NiagaraFloat=&FN_FrontEndFloatModifier(Gameplay=4, Front End=3)
    Orientation Method=::NewEnumerator1
    Override Quat=false
    TransformFacing=false
    TransformUp=false
    Up Direction:Vector3f=<1 0 0>
   ) {enabled=false, module=UpdateMeshOrientation001}
   /SetVariables(Particles.MeshOrientation=&OrientMeshToVector(
     Look At Direction=&FN_FrontEndVector3Modifier(
      Gameplay=&Multiply_VectorByFloat(
       Vector=&NormalizeVector(
        Vector To Normalize=&NegateVector(Vector:Vector3f=&TransformVector(
          Vector=&Multiply_Vector(A=$Engine.Owner.Velocity, B=<0.3 0.3 1>)
          Source Space=::World
          Destination Space=::Simulation
         ))
        Safe Normalize=true
        Safe Fallback Threshold=0
        Safe Fallback Vector=<0 0 0.001>
       )
       Float=-1
      )
      Front End=<0 0.25 1>
     )
     Reference Axis Vector=<1 0 0>
    )) {isSetParameters=true, module=SetVariables_#13}
   &SolveForcesAndVelocity(Clamp Velocity=false, Limit Acceleration=false)
   &ScaleMeshSize(Scale Factor=&VectorFromCurve(
     VectorCurve=DataInterface() {propertyValues {
       XCurve {keys [
         {arriveTangent=4.4809, interpMode=RCIM_Cubic, leaveTangent=4.4809, tangentMode=RCTM_User, time=0.1, value=0}
         <at 0.4 0.475>
         {arriveTangent=0.875, interpMode=RCIM_Cubic, leaveTangent=0.875, time=0.6, value=0.475}
         {tangentMode=RCTM_User, time=1, value=1}
        ]}
       YCurve {keys [
         {arriveTangent=4.4809, interpMode=RCIM_Cubic, leaveTangent=4.4809, tangentMode=RCTM_User, time=0.1, value=0}
         <at 0.4 0.475>
         {arriveTangent=0.875, interpMode=RCIM_Cubic, leaveTangent=0.875, time=0.6, value=0.475}
         {tangentMode=RCTM_User, time=1, value=1}
        ]}
       ZCurve {keys [{tangentMode=RCTM_User, time=0, value=-0.2}, {tangentMode=RCTM_User, time=1, value=1}]}
      }}
     CurveIndex=$Emitter.Age
    ))
   &UpdateMeshOrientation(
    Rotation Coordinate Space=::NewEnumerator2
    Rotation Rate=&Lerp_Float(A=0.14, B=0.35, Alpha=$System.FN_Integrator001.Value)
    Rotation Vector=<0 0 1>
   ) {enabled=false}
   &ScaleColor(
    Scale Alpha=&FloatFromCurve(
     FloatCurve=DataInterface() {propertyValues {Curve {keys [
         {arriveTangent=4.0516, interpMode=RCIM_Cubic, leaveTangent=4.0516, tangentMode=RCTM_User, time=0, value=0}
         {arriveTangent=0.0961, interpMode=RCIM_Cubic, leaveTangent=0.0961, tangentMode=RCTM_User, time=0.25, value=1}
         {arriveTangent=0.2272, interpMode=RCIM_Cubic, leaveTangent=0.2272, tangentMode=RCTM_User, time=1, value=1}
        ]}}}
     Scale Curve=&Lerp_Float(A=0.2, B=0.5, Alpha=$System.Dive)
    )
    ScaleA=true
    ScaleRGB=false
   )
   &FN_Light_Attributes(
    Color=<rgba 1 0.2 0 1>
    Create Color Attribute=true
    Create Exponent Attribute=true
    Create Radius Attribute=true
    Exponent=2
    Radius=90
    Scale Color=&FloatFromCurve(
     FloatCurve=DataInterface() {propertyValues {Curve {keys [
         {interpMode=RCIM_Cubic, tangentMode=RCTM_User, time=1.1, value=0}
         <at 1.4 0.25>
         {arriveTangent=2.7482, interpMode=RCIM_Cubic, leaveTangent=2.7482, tangentMode=RCTM_User, time=1.9, value=0.25}
         {interpMode=RCIM_Cubic, tangentMode=RCTM_User, time=2.4, value=1}
        ]}}}
     CurveIndex=$Emitter.Age
     Scale Curve=1.29
    )
   )
   &CameraOffset(Camera Offset Amount=&Lerp_Float(A=0, B=-18, Alpha=$System.FN_Integrator001.Value))
   &FN_Mesh_Attributes(Color=<rgba 4 1.6829 0.4468 1>, Create Color Attribute=true, Module Output=::NewEnumerator1)
   &FN_Mesh_Attributes(
    Color=&MakeLinearColorFromVectorAndFloat(
     "Vector (RGB)"=&Multiply_Vector(
      A=&MakeVectorFromLinearColorRGB(Color=<rgba 0.0282 0.0049 0.0365 1>)
      B=&Multiply_VectorByFloat(Vector=<1 1 1>, Float=10)
     )
     "Float (Alpha)"=&Multiply_Float(A=1, B=$Particles.AlphaAll)
    )
    Create Color Attribute=true
    Create Position Attribute=true
    Module Output=::NewEnumerator1
    Position:NiagaraPosition=<0 0 49>
   ) {module=FN_Mesh_Attributes001}
   &FN_Mesh_Attributes(
    Color=&MakeLinearColorFromVectorAndFloat(
     "Vector (RGB)"=&Multiply_Vector(
      A=&MakeVectorFromLinearColorRGB(Color=<rgba 0.0282 0.0049 0.0365 1>)
      B=&Multiply_VectorByFloat(Vector=<1 1 1>, Float=15)
     )
     "Float (Alpha)"=&Multiply_Float(A=1, B=$Particles.AlphaAll)
    )
    Create Color Attribute=true
    Create Scale Attribute=true
    Module Output=::NewEnumerator1
    Scale=&Multiply_VectorByFloat(Vector=<1.3 1.3 0.6>, Float=1)
   ) {module=FN_Mesh_Attributes002}
   &FN_Mesh_Attributes(
    Color=&MakeLinearColorFromVectorAndFloat(
     "Vector (RGB)"=&Multiply_Vector(
      A=&MakeVectorFromLinearColorRGB(Color=<rgba 0.2663 0.0009 1 0.1>)
      B=&Multiply_VectorByFloat(Vector=<1 1 1>, Float=1)
     )
     "Float (Alpha)"=&Multiply_Float(A=2, B=$Particles.AlphaAll)
    )
    Create Color Attribute=true
    Create Position Attribute=true
    Create Scale Attribute=true
    Module Output=::NewEnumerator1
    Position:NiagaraPosition=<0 0 66.5>
    Scale=&Multiply_VectorByFloat(Vector=<1.3 1.3 0.6>, Float=1)
   ) {module=FN_Mesh_Attributes003}
   &DynamicMaterialParameters(
    Index 0 Param 1=3
    Index 0 Param 2=0.4
    Index 2 Param 1=&Lerp_Float(
     A=HlslExpression() {hlslExpression=Particles.Age * 8.0 +5.4}
     B=HlslExpression() {hlslExpression=Particles.Age * 16.0 +5.4}
     Alpha=$System.SmoothDiveAlpha
    )
    Module.Index0Float0Write=true
    Module.Index0Float1Write=true
    Module.Index0Float2Write=true
    Module.Index0Float3Write=true
    Module.Index2Float0Write=true
    Module.Index2Float1Write=false
    Module.Index2Float2Write=false
    Module.Index2Float3Write=false
    Module.Index3Float0Write=false
    Module.Index3Float1Write=false
    Module.Index3Float2Write=false
    Module.Index3Float3Write=false
    Write Parameter Index 2=::NewEnumerator1
    Write Parameter Index 3=::NewEnumerator1
   ) {module=DynamicMaterialParameters001}
  ]
  rendererClasses [&NiagaraLightRendererProperties, &NiagaraMeshRendererProperties]
  renderers [
   {rendererClass=&NiagaraLightRendererProperties, rendererIndex=0}
   {rendererClass=&NiagaraMeshRendererProperties, rendererIndex=1}
   {rendererClass=&NiagaraMeshRendererProperties, rendererIndex=2}
   {rendererClass=&NiagaraMeshRendererProperties, rendererIndex=3}
   {rendererClass=&NiagaraMeshRendererProperties, rendererIndex=4}
  ]
 }
 DarkRibbon {
  emitterUpdateScript [
   &EmitterState(
    Inactive Response=::NewEnumerator1
    Life Cycle Mode=::NewEnumerator1
    Loop Behavior=::NewEnumerator1
    Loop Duration Mode=::NewEnumerator1
    UseLoopDelay=false
   ) {module=EmitterLifeCycle}
   &SpawnBurst_Instantaneous(
    Spawn Count=&Multiply_Int(A=$System.ChainNumber, B=$System.ParticlePerRibbon)
    Use Spawn Probability=false
    UseLoopCountLimit=false
   )
  ]
  particleSpawnScript [
   &InitializeParticle(Lifetime=5, Sprite Size Max=<20 20>, Sprite Size Mode=::NewEnumerator2, UsePositionOffset=false, Write Material Random=true)
   &PartitionParticles(
    Number Of Partitions=$System.ChainNumber
    Particles Per Partition:NiagaraInt32=$System.ParticlePerRibbon
    Partition Mode:ENiagara_PartitionMode_v2=::NewEnumerator1
   )
   /SetVariables(Particles.NormalizedChainPosition:NiagaraFloat=$Particles.PartitionParticles.Normalized Index In Partition) {isSetParameters=true, module=SetVariables_#14}
   /SetVariables(Particles.LengthMulti=0.85) {isSetParameters=true}
  ]
  particleUpdateScript [
   &ParticleState(Kill Particles When Lifetime Has Elapsed=false)
   &SkeletalMeshLocation(
    Bone / Socket Index=$Particles.PartitionParticles.Partition
    Bone Sampling Mode=::NewEnumerator4
    MaskSpawnGroup=false
    Position Sampling=::NewEnumerator2
    Sampled Position Offset=&Lerp_Vector(
     A=&Lerp_Vector(A=<0 -5 2>, B=<0 5 -2>, Alpha=&MakeFloatFromInt(Int=$Particles.PartitionParticles.Partition))
     B=&Lerp_Vector(A=<0 0 0>, B=<0 0 0>, Alpha=&MakeFloatFromInt(Int=$Particles.PartitionParticles.Partition))
     Alpha=$User.ParaGlideLeanAlpha
    )
    Skeletal Mesh=$User.SkeletalMesh
   ) {module=SkeletalMeshLocation001}
   &PendulumSetup(
    Module.Enable Spring Driver=true
    Pendulum Length:NiagaraFloat=&Lerp_Float(
     A=0
     B=&Multiply_Float(A=$System.ChainLength, B=$Particles.Initial.LengthMulti)
     Alpha=&FloatFromCurve(
      FloatCurve=DataInterface() {propertyValues {Curve {keys [<at 0 0>, {arriveTangent=-0.0703, interpMode=RCIM_Cubic, leaveTangent=-0.0703, tangentMode=RCTM_User, time=1, value=1}]}}}
      CurveIndex=$Particles.NormalizedChainPosition
      Scale Curve=&FloatFromCurve(
       FloatCurve=DataInterface() {propertyValues {Curve {keys [
           {arriveTangent=0.6092, interpMode=RCIM_Cubic, leaveTangent=0.6092, tangentMode=RCTM_User, time=0, value=0}
           {tangentMode=RCTM_User, time=1, value=1}
          ]}}}
       CurveIndex=$Emitter.Age
      )
     )
    )
    Pendulum Pivot:Vector3f=&AddVectorToPosition(
     Position=&AddVectorToPosition(Position=$Particles.SkeletalMeshLocation001.SampledPosition, Vector=<0 0 0>)
     Vector=&TransformVector(Vector=<0 0 0>, Source Space=::World, Destination Space=::Simulation)
    )
    Pendulum Pivot Offset Coordinate Space:ENiagaraCoordinateSpace=::Local
    Pendulum Rest Axis:Vector3f=&TransformVector(
     Vector=&Add_Vector(
      A=&Lerp_Vector(
       A=&Add_Vector(
        A=&Multiply_VectorByFloat(
         Vector=&NormalizeVector(Vector To Normalize=$Engine.Owner.Velocity, Safe Normalize=true, Safe Fallback Threshold=0, Safe Fallback Vector=<0 0 0.001>)
         Float=-0.5
        )
        B=&TransformVector(
         Vector=&RotateVectorByQuaternion(
          Vector To Rotate=&Multiply_VectorByFloat(
           Vector=<0.5 1 0>
           Float=&Lerp_Float(A=1, B=-1, Alpha=&MakeFloatFromInt(Int=$Particles.PartitionParticles.Partition))
          )
          Quaternion=$Particles.SkeletalMeshLocation001.SampledRotation
         )
         Destination Space=::World
        )
       )
       B=&Lerp_Vector(
        A=&Multiply_VectorByFloat(
         Vector=&NormalizeVector(Vector To Normalize=$Engine.Owner.Velocity, Safe Normalize=true, Safe Fallback Threshold=0, Safe Fallback Vector=<0 0 0.001>)
         Float=-1
        )
        B=<0 0 0.1>
        Alpha=$System.SmoothDiveAlpha
       )
       Alpha=&Lerp_Float(
        A=&FloatFromCurve(
         FloatCurve=DataInterface() {propertyValues {Curve {keys [
             {arriveTangent=1.8207, interpMode=RCIM_Cubic, leaveTangent=1.8207, tangentMode=RCTM_User, time=0, value=0}
             {arriveTangent=-0.078, interpMode=RCIM_Cubic, leaveTangent=-0.078, tangentMode=RCTM_User, time=1, value=1}
            ]}}}
         CurveIndex=$Particles.NormalizedChainPosition
        )
        B=1
        Alpha=&FN_FrontEndFloatModifier(Gameplay=0, Front End=1)
       )
      )
      B=&TransformVector(
       Vector=&FN_FrontEndVector3Modifier(Gameplay=<0 0 1>, Front End=&Lerp_Vector(A=<0 -0.1 0.2>, B=<0 -0.2 0.2>, Alpha=$System.SmoothDiveAlpha))
       Destination Space=::World
      )
     )
     Source Space=::World
     Destination Space=::Simulation
    )
    "Spring Driven Constraint (Tightness)"=30
   )
   &SolveForcesAndVelocity(Clamp Velocity=false, Limit Acceleration=false)
   &PendulumConstraint()
   &CalculateAccurateVelocity()
   /SetVariables(Particles.RibbonWidth=&FloatFromCurve(
     FloatCurve=DataInterface() {propertyValues {Curve {keys [
         {arriveTangent=8.0134, interpMode=RCIM_Cubic, leaveTangent=8.0134, tangentMode=RCTM_User, time=0, value=0.7}
         {arriveTangent=0.3471, interpMode=RCIM_Cubic, leaveTangent=0.3471, tangentMode=RCTM_User, time=0.1, value=1.2}
         {arriveTangent=-0.7173, interpMode=RCIM_Cubic, leaveTangent=-0.7173, tangentMode=RCTM_User, time=1, value=2}
        ]}}}
     CurveIndex=$Particles.NormalizedChainPosition
     Scale Curve=&Lerp_Float(A=40, B=30, Alpha=$System.SmoothDiveAlpha)
    )) {isSetParameters=true, module=SetVariables_#15}
   &DynamicMaterialParameters(
    Index 0 Param 1=1
    Index 0 Param 2=&Add_Float(
     A=&Multiply_Float(A=$Engine.System.Age, B=1.2)
     B=&Multiply_FloatByInt(Float=3.14, Integer=$Particles.PartitionParticles.Partition)
    )
    Index 0 Param 3=1
    Index 0 Param 4=&Add_Float(
     A=&Multiply_Float(A=$Engine.System.Age, B=1.2)
     B=&Multiply_FloatByInt(Float=1.6902, Integer=$Particles.PartitionParticles.Partition)
    )
    Index 1 Param 1=1
    Index 1 Param 2=&Add_Float(
     A=&Multiply_Float(A=$Engine.System.Age, B=1.5786)
     B=&Multiply_FloatByInt(Float=1.6902, Integer=$Particles.PartitionParticles.Partition)
    )
    Index 1 Param 3=1
    Index 3 Param 1=1
    Index 3 Param 2=1
    Index 3 Param 3=1
    Module.Index0Param0Write=true
    Module.Index0Param1Write=true
    Module.Index0Param2Write=true
    Module.Index0Param3Write=true
    Module.Index1Param0Write=true
    Module.Index1Param1Write=true
    Module.Index1Param2Write=true
    Module.Index1Param3Write=true
    Module.Index3Param0Write=true
    Module.Index3Param1Write=true
    Module.Index3Param2Write=true
    Module.Index3Param3Write=true
    Write Parameter Index 1=true
    Write Parameter Index 3=true
   )
   &self:Ribbon3Location_03(
    Amplitude:Vector3f=&Multiply_VectorByFloat(
     Vector=&Lerp_Vector(A=<1 1 0>, B=<4 4 0>, Alpha=$User.ParaGlideLeanAlpha)
     Float=&FloatFromCurve(
      FloatCurve=DataInterface() {propertyValues {Curve {keys [
          {interpMode=RCIM_Cubic, tangentMode=RCTM_User, time=0, value=0}
          {arriveTangent=-2.3116, interpMode=RCIM_Cubic, leaveTangent=-2.3116, tangentMode=RCTM_User, time=0.4, value=1}
          <at 1 0>
         ]}}}
      CurveIndex=$Particles.PartitionParticles.Normalized Index In Partition
      Scale Curve=&Lerp_Float(A=1, B=1, Alpha=$System.SmoothDiveAlpha)
     )
    )
    Amplitude2:Vector3f=&Lerp_Vector(A=<5 5 0>, B=<40 40 0>, Alpha=$User.ParaGlideLeanAlpha)
    Amplitude3:Vector3f=&Lerp_Vector(A=<3 3 0>, B=<12 12 0>, Alpha=$User.ParaGlideLeanAlpha)
    Partition:NiagaraFloat=&MakeFloatFromInt(Int=$Particles.PartitionParticles.Partition)
    Scale:Vector3f=&Multiply_VectorByFloat(Vector=<1 1 1>, Float=0.5)
    Scale2:Vector3f=&Multiply_VectorByFloat(Vector=<0.8 0.8 1>, Float=0.5)
    Scale3:Vector3f=&Multiply_VectorByFloat(Vector=<1 1 1>, Float=0.5)
    Time:NiagaraFloat=&Add_Float(
     A=&Multiply_Float(A=$Engine.System.Age, B=-1.5)
     B=$Particles.PartitionParticles.Normalized Index In Partition
    )
   ) {module=Ribbon3Location_03}
   /SetVariables(
    Particles.DarkRibbon_Color:LinearColor=&MakeLinearColorFromVectorAndFloat(
     "Vector (RGB)"=&Multiply_VectorByFloat(
      Vector=&MakeVectorFromLinearColorRGB(Color=&ColorFromCurve(
        ColorCurve=DataInterface() {propertyValues {
          AlphaCurve {keys [<at 0 1 RCIM_Cubic>,<at 1 1 RCIM_Cubic>]}
          BlueCurve {keys [<at 0 1.3>,<at 0.2787 2>,<at 0.4745 1.5>,<at 0.7559 0.1673>,<at 0.8762 0>,<at 1 0>]}
          GreenCurve {keys [<at 0 0.6828>,<at 0.2787 0.2779>,<at 0.4745 0>,<at 0.7559 0>,<at 0.8762 0.5146>,<at 1 0.8811>]}
          RedCurve {keys [<at 0 0>,<at 0.2787 0>,<at 0.4745 0.7049>,<at 0.7559 2>,<at 0.8762 2>,<at 1 2>]}
         }}
        CurveIndex=&Waveform(
         Global Input Position=&Multiply_FloatByInt(Float=$Particles.NormalizedChainPosition, Integer=1)
         Global Phase=&Add_Float(
          A=&Multiply_Float(A=$Engine.System.Age, B=-1)
          B=&Multiply_FloatByInt(Float=34, Integer=$Particles.PartitionParticles.Partition)
         )
         "Waveform [1]"=::NewEnumerator12
         "[1] Frequency"=0.8094
         "[1] Amplitude Min/Max"=<0 1>
        )
       ))
      Float=&Lerp_Float(A=0.4, B=&FN_FrontEndFloatModifier(Gameplay=0.9, Front End=0.5), Alpha=$System.SmoothDiveAlpha)
     )
     "Float (Alpha)"=&FloatFromCurve(
      FloatCurve=DataInterface() {propertyValues {Curve {keys [
          <at 0 1 RCIM_Cubic>
          {arriveTangent=-2.1381, interpMode=RCIM_Cubic, leaveTangent=-2.1381, tangentMode=RCTM_User, time=0.4255, value=0.5}
          <at 0.949 0 RCIM_Cubic>
         ]}}}
      CurveIndex=$Particles.NormalizedChainPosition
      Scale Curve=&FloatFromCurve(
       FloatCurve=DataInterface() {propertyValues {Curve {keys [{interpMode=RCIM_Cubic, tangentMode=RCTM_User, time=0, value=0}, {interpMode=RCIM_Cubic, tangentMode=RCTM_User, time=1, value=1}]}}}
       CurveIndex=$Engine.System.Age
       Scale Curve=2.5
      )
     )
    )
    Particles.DarkRibbon_Width:NiagaraFloat=&Multiply_Float(A=$Particles.RibbonWidth, B=1)
    Particles.DynamicMaterialParameter8:Vector4f=&MakeVector4(X=1, Y=1, Z=&Lerp_Float(A=1.5, B=2, Alpha=$System.SmoothDiveAlpha), W=0)
    Particles.RibbonFacing=/Niagara/DynamicInputs/Vectors/CrossProduct(Vector A:Vector3f=$Transient.PhysicsPendulumAxis, Vector B:Vector3f=<0 0 1>, Normalize Incoming Vectors:NiagaraBool=true)
   ) {isSetParameters=true, module=SetVariables_#16}
  ]
  rendererClasses [&NiagaraRibbonRendererProperties]
  renderers [{rendererClass=&NiagaraRibbonRendererProperties}]
 }
 Flare {
  emitterUpdateScript [
   &EmitterState(
    Inactive Response=::NewEnumerator1
    Life Cycle Mode=::NewEnumerator1
    Loop Behavior=::NewEnumerator1
    Loop Duration Mode=::NewEnumerator1
    UseLoopDelay=false
   )
   &SpawnBurst_Instantaneous(Spawn Count=4, Use Spawn Probability=false)
  ]
  particleSpawnScript [
   &FN_SetSpawnGroup(Number of Spawn Groups=2)
   &PartitionParticles()
   &InitializeParticle(
    Color=<rgba 0.7627 0.0267 1 1>
    Sprite Size Mode=::NewEnumerator3
    Uniform Sprite Size=100
    UsePositionOffset=false
    Write Material Random=true
   )
   /SetVariables(Particles.FlareColor=<rgba 8.6296 0.7434 10 1>) {isSetParameters=true}
   /SetVariables(Particles.FlareSize:Vector2f=&Vector2DFromFloat(Value=&SelectFloatFromArray(Direct Array Index=$Particles.SpawnGroup, Float Selection Array=DataInterface() {propertyValues {FloatData [50,50,30,30]}}))) {isSetParameters=true}
  ]
  particleUpdateScript [
   &FN_DivingAttributes_v2(Accumulate Dive=0.1, Accumulate Rest=-0.1, ParaGlideLeanAlpha:NiagaraFloat=$User.ParaGlideLeanAlpha)
   /SetVariables(Particles.Dive01:NiagaraFloat=&Multiply_Float(A=1, B=$Particles.Dive)) {isSetParameters=true}
   &ParticleState(Kill Particles When Lifetime Has Elapsed=false, Let Infinitely Lived Particles Die When Emitter Deactivates=true)
   &ScaleColor(
    Scale Alpha=&Multiply_Float(
     A=&FNTimeOfDayScalar(Night Scale=0.2, Morning Scale=0.5, Evening Scale=0.5)
     B=&Multiply_Float(
      A=&FN_FrontEndFloatModifier(Gameplay=0.5, Front End=1)
      B=&Lerp_Float(A=1, B=0.5, Alpha=$Particles.Dive01)
     )
    )
    ScaleA=true
    ScaleRGB=true
   )
   &ScaleSpriteSize(Scale Factor=&Vector2DFromFloat(Value=&Lerp_Float(A=1.5, B=0.7, Alpha=$Particles.Dive01)))
   /SetVariables(Particles.Color001:LinearColor=&MultiplyLinearColor(
     Linear Color=$Particles.FlareColor
     Scale Factor=&MakeFloat4FromVectorAndFloat(
      "Vector (RGB)"=&VectorFromFloat(Value=&Sine(Normalized Angle=$Emitter.Age, Period=0.6, Scale=0.5, Bias=0.7))
      "Float (Alpha)"=&Multiply_Float(
       A=&Multiply_Float(
        A=&FNTimeOfDayScalar(Night Scale=0.2, Morning Scale=0.5, Evening Scale=0.5)
        B=&FN_FrontEndFloatModifier(Gameplay=0.5, Front End=1)
       )
       B=&Lerp_Float(A=1, B=0.5, Alpha=$Particles.Dive01)
      )
     )
    )) {isSetParameters=true}
   /SetVariables(Particles.SpriteSize001:Vector2f=&Multiply_Vector2(
     A=$Particles.FlareSize
     B=&Vector2DFromFloat(Value=&Lerp_Float(A=0.4, B=0.4, Alpha=$Particles.Dive01))
    )) {isSetParameters=true}
   &SkeletalMeshLocation(
    Bone / Socket Index=$Particles.SpawnGroup
    Bone Sampling Mode=::NewEnumerator4
    Inherit Velocity=false
    MaskSpawnGroup=false
    Sampled Position Offset=&Lerp_Vector(
     A=&Lerp_Vector(A=<0 -10 5>, B=<0 -10 5>, Alpha=&MakeFloatFromInt(Int=$Particles.PartitionParticles.Partition))
     B=&Lerp_Vector(A=<4 5 0>, B=<-4 5 0>, Alpha=&MakeFloatFromInt(Int=$Particles.PartitionParticles.Partition))
     Alpha=$System.Dive
    )
    Sampled Position Offset Coordinate Space:ENiagaraMeshTransforms=::NewEnumerator2
    Skeletal Mesh=$User.SkeletalMesh
   )
   &CameraOffset(Camera Offset Amount=&FN_FrontEndFloatModifier(Gameplay=30, Front End=30))
  ]
  rendererClasses [&NiagaraSpriteRendererProperties]
  renderers [
   {rendererClass=&NiagaraSpriteRendererProperties}
   {rendererClass=&NiagaraSpriteRendererProperties, rendererIndex=1}
  ]
 }
 GoldRibbon {
  emitterUpdateScript [
   &EmitterState(
    Inactive Response=::NewEnumerator1
    Life Cycle Mode=::NewEnumerator1
    Loop Behavior=::NewEnumerator1
    Loop Duration Mode=::NewEnumerator1
    UseLoopDelay=false
   ) {module=EmitterLifeCycle}
   &SpawnBurst_Instantaneous(
    Spawn Count=&Multiply_Int(A=$System.ChainNumber, B=$System.ParticlePerRibbon)
    Use Spawn Probability=false
    UseLoopCountLimit=false
   )
  ]
  particleSpawnScript [
   &InitializeParticle(Lifetime=5, Sprite Size Max=<20 20>, Sprite Size Mode=::NewEnumerator2, UsePositionOffset=false, Write Material Random=true)
   &PartitionParticles(
    Number Of Partitions=$System.ChainNumber
    Particles Per Partition:NiagaraInt32=$System.ParticlePerRibbon
    Partition Mode:ENiagara_PartitionMode_v2=::NewEnumerator1
   )
   /SetVariables(Particles.NormalizedChainPosition:NiagaraFloat=$Particles.PartitionParticles.Normalized Index In Partition) {isSetParameters=true, module=SetVariables_#14}
   /SetVariables(Particles.LengthMulti=0.68) {isSetParameters=true}
  ]
  particleUpdateScript [
   &ParticleState(Kill Particles When Lifetime Has Elapsed=false)
   &SkeletalMeshLocation(
    Bone / Socket Index=$Particles.PartitionParticles.Partition
    Bone Sampling Mode=::NewEnumerator4
    MaskSpawnGroup=false
    Position Sampling=::NewEnumerator2
    Sampled Position Offset=&Lerp_Vector(
     A=&Lerp_Vector(A=<0 -20 -2>, B=<0 20 2>, Alpha=&MakeFloatFromInt(Int=$Particles.PartitionParticles.Partition))
     B=&Lerp_Vector(A=<0 0 0>, B=<0 0 0>, Alpha=&MakeFloatFromInt(Int=$Particles.PartitionParticles.Partition))
     Alpha=$User.ParaGlideLeanAlpha
    )
    Skeletal Mesh=$User.SkeletalMesh
   ) {module=SkeletalMeshLocation001}
   &PendulumSetup(
    Module.Enable Spring Driver=true
    Pendulum Length:NiagaraFloat=&Lerp_Float(
     A=0
     B=&Multiply_Float(A=$System.ChainLength, B=$Particles.Initial.LengthMulti)
     Alpha=&FloatFromCurve(
      FloatCurve=DataInterface() {propertyValues {Curve {keys [<at 0 0>, {arriveTangent=-0.0703, interpMode=RCIM_Cubic, leaveTangent=-0.0703, tangentMode=RCTM_User, time=1, value=1}]}}}
      CurveIndex=$Particles.NormalizedChainPosition
      Scale Curve=&FloatFromCurve(
       FloatCurve=DataInterface() {propertyValues {Curve {keys [
           {arriveTangent=0.6092, interpMode=RCIM_Cubic, leaveTangent=0.6092, tangentMode=RCTM_User, time=0, value=0}
           {tangentMode=RCTM_User, time=1, value=1}
          ]}}}
       CurveIndex=$Emitter.Age
      )
     )
    )
    Pendulum Pivot:Vector3f=&AddVectorToPosition(
     Position=&AddVectorToPosition(Position=$Particles.SkeletalMeshLocation001.SampledPosition, Vector=<0 0 0>)
     Vector=&TransformVector(Vector=<0 0 0>, Source Space=::World, Destination Space=::Simulation)
    )
    Pendulum Pivot Offset Coordinate Space:ENiagaraCoordinateSpace=::Local
    Pendulum Rest Axis:Vector3f=&TransformVector(
     Vector=&Add_Vector(
      A=&Lerp_Vector(
       A=&Add_Vector(
        A=&Multiply_VectorByFloat(
         Vector=&NormalizeVector(Vector To Normalize=$Engine.Owner.Velocity, Safe Normalize=true, Safe Fallback Threshold=0, Safe Fallback Vector=<0 0 0.001>)
         Float=-0.5
        )
        B=&TransformVector(
         Vector=&RotateVectorByQuaternion(
          Vector To Rotate=&Multiply_VectorByFloat(
           Vector=<0.5 1 0>
           Float=&Lerp_Float(A=1, B=-1, Alpha=&MakeFloatFromInt(Int=$Particles.PartitionParticles.Partition))
          )
          Quaternion=$Particles.SkeletalMeshLocation001.SampledRotation
         )
         Destination Space=::World
        )
       )
       B=&Lerp_Vector(
        A=&Multiply_VectorByFloat(
         Vector=&NormalizeVector(Vector To Normalize=$Engine.Owner.Velocity, Safe Normalize=true, Safe Fallback Threshold=0, Safe Fallback Vector=<0 0 0.001>)
         Float=-1
        )
        B=<0 0 0.1>
        Alpha=$System.SmoothDiveAlpha
       )
       Alpha=&Lerp_Float(
        A=&FloatFromCurve(
         FloatCurve=DataInterface() {propertyValues {Curve {keys [
             {arriveTangent=1.8207, interpMode=RCIM_Cubic, leaveTangent=1.8207, tangentMode=RCTM_User, time=0, value=0}
             {arriveTangent=-0.078, interpMode=RCIM_Cubic, leaveTangent=-0.078, tangentMode=RCTM_User, time=1, value=1}
            ]}}}
         CurveIndex=$Particles.NormalizedChainPosition
        )
        B=1
        Alpha=&FN_FrontEndFloatModifier(Gameplay=0, Front End=1)
       )
      )
      B=&TransformVector(
       Vector=&FN_FrontEndVector3Modifier(Gameplay=<0 0 1>, Front End=&Lerp_Vector(A=<0 -0.1 0.2>, B=<0 -0.2 0.2>, Alpha=$System.SmoothDiveAlpha))
       Destination Space=::World
      )
     )
     Source Space=::World
     Destination Space=::Simulation
    )
    "Spring Driven Constraint (Tightness)"=30
   )
   &SolveForcesAndVelocity(Clamp Velocity=false, Limit Acceleration=false)
   &PendulumConstraint()
   &CalculateAccurateVelocity()
   /SetVariables(Particles.RibbonWidth=&FloatFromCurve(
     FloatCurve=DataInterface() {propertyValues {Curve {keys [
         {arriveTangent=8.0134, interpMode=RCIM_Cubic, leaveTangent=8.0134, tangentMode=RCTM_User, time=0, value=0.7}
         {arriveTangent=0.3471, interpMode=RCIM_Cubic, leaveTangent=0.3471, tangentMode=RCTM_User, time=0.1, value=1.2}
         {arriveTangent=-0.7173, interpMode=RCIM_Cubic, leaveTangent=-0.7173, tangentMode=RCTM_User, time=1, value=2}
        ]}}}
     CurveIndex=$Particles.NormalizedChainPosition
     Scale Curve=&Lerp_Float(A=40, B=30, Alpha=$System.SmoothDiveAlpha)
    )) {isSetParameters=true, module=SetVariables_#15}
   &DynamicMaterialParameters(
    Index 0 Param 1=1
    Index 0 Param 2=&Add_Float(
     A=&Multiply_Float(A=$Engine.System.Age, B=1.2)
     B=&Multiply_FloatByInt(Float=3.14, Integer=$Particles.PartitionParticles.Partition)
    )
    Index 0 Param 3=1
    Index 0 Param 4=&Add_Float(
     A=&Multiply_Float(A=$Engine.System.Age, B=1.2)
     B=&Multiply_FloatByInt(Float=1.6902, Integer=$Particles.PartitionParticles.Partition)
    )
    Index 1 Param 1=1
    Index 1 Param 2=&Add_Float(
     A=&Multiply_Float(A=$Engine.System.Age, B=1.5786)
     B=&Multiply_FloatByInt(Float=1.6902, Integer=$Particles.PartitionParticles.Partition)
    )
    Index 1 Param 3=1
    Index 3 Param 1=1
    Index 3 Param 2=1
    Index 3 Param 3=1
    Module.Index0Param0Write=true
    Module.Index0Param1Write=true
    Module.Index0Param2Write=true
    Module.Index0Param3Write=true
    Module.Index1Param0Write=true
    Module.Index1Param1Write=true
    Module.Index1Param2Write=true
    Module.Index1Param3Write=true
    Module.Index3Param0Write=true
    Module.Index3Param1Write=true
    Module.Index3Param2Write=true
    Module.Index3Param3Write=true
    Write Parameter Index 1=true
    Write Parameter Index 3=true
   )
   &self:Ribbon3Location_03(
    Amplitude:Vector3f=&Multiply_VectorByFloat(
     Vector=&Lerp_Vector(A=<1 1 0>, B=<4 4 0>, Alpha=$User.ParaGlideLeanAlpha)
     Float=&FloatFromCurve(
      FloatCurve=DataInterface() {propertyValues {Curve {keys [
          {interpMode=RCIM_Cubic, tangentMode=RCTM_User, time=0, value=0}
          {arriveTangent=-2.3116, interpMode=RCIM_Cubic, leaveTangent=-2.3116, tangentMode=RCTM_User, time=0.4, value=1}
          <at 1 0>
         ]}}}
      CurveIndex=$Particles.PartitionParticles.Normalized Index In Partition
      Scale Curve=&Lerp_Float(A=1, B=1, Alpha=$System.SmoothDiveAlpha)
     )
    )
    Amplitude2:Vector3f=&Lerp_Vector(A=<5 5 0>, B=<40 40 0>, Alpha=$User.ParaGlideLeanAlpha)
    Amplitude3:Vector3f=&Lerp_Vector(A=<3 3 0>, B=<12 12 0>, Alpha=$User.ParaGlideLeanAlpha)
    Partition:NiagaraFloat=&MakeFloatFromInt(Int=$Particles.PartitionParticles.Partition)
    Scale:Vector3f=&Multiply_VectorByFloat(Vector=<1 1 1>, Float=0.5)
    Scale2:Vector3f=&Multiply_VectorByFloat(Vector=<0.8 0.8 1>, Float=0.5)
    Scale3:Vector3f=&Multiply_VectorByFloat(Vector=<1 1 1>, Float=0.5)
    Time:NiagaraFloat=&Add_Float(
     A=&Multiply_Float(A=$Engine.System.Age, B=-1.5)
     B=$Particles.PartitionParticles.Normalized Index In Partition
    )
   ) {module=Ribbon3Location_03}
   /SetVariables(
    Particles.DynamicMaterialParameter9:Vector4f=&MakeVector4(
     X=&Multiply_Float(
      A=$Engine.System.Age
      B=&Lerp_Float(A=-2, B=2, Alpha=&MakeFloatFromInt(Int=$Particles.Initial.PartitionParticles.Partition))
     )
     Y=&Lerp_Float(A=-3, B=3, Alpha=&MakeFloatFromInt(Int=$Particles.Initial.PartitionParticles.Partition))
     Z=&Multiply_Float(A=1, B=&Lerp_Float(A=0.4, B=1, Alpha=$System.SmoothDiveAlpha))
     W=&FloatFromCurve(
      FloatCurve=DataInterface() {propertyValues {Curve {keys [{arriveTangent=-2.2554, interpMode=RCIM_Cubic, leaveTangent=-2.2554, tangentMode=RCTM_User, time=0, value=1}, <at 1 0>]}}}
      CurveIndex=$Particles.PartitionParticles.Normalized Index In Partition
     )
    )
    Particles.GoldRibbbon_Width:NiagaraFloat=&Multiply_Float(
     A=$Particles.RibbonWidth
     B=&FloatFromCurve(
      FloatCurve=DataInterface() {propertyValues {Curve {keys [
          {arriveTangent=0.357, interpMode=RCIM_Cubic, leaveTangent=0.357, tangentMode=RCTM_User, time=0, value=0.3}
          {arriveTangent=-0.0652, interpMode=RCIM_Cubic, leaveTangent=-0.0652, tangentMode=RCTM_User, time=0.2, value=1}
          {arriveTangent=-0.1821, interpMode=RCIM_Cubic, leaveTangent=-0.1821, tangentMode=RCTM_User, time=0.7, value=0}
         ]}}}
      CurveIndex=$Particles.NormalizedChainPosition
      Scale Curve=&Lerp_Float(A=0.2, B=0.4, Alpha=$System.SmoothDiveAlpha)
     )
    )
   ) {isSetParameters=true, module=SetVariables_#17}
  ]
  rendererClasses [&NiagaraRibbonRendererProperties]
  renderers [{rendererClass=&NiagaraRibbonRendererProperties}]
 }
 LightRibbon {
  emitterUpdateScript [
   &EmitterState(
    Inactive Response=::NewEnumerator1
    Life Cycle Mode=::NewEnumerator1
    Loop Behavior=::NewEnumerator1
    Loop Duration Mode=::NewEnumerator1
    UseLoopDelay=false
   ) {module=EmitterLifeCycle}
   &SpawnBurst_Instantaneous(
    Spawn Count=&Multiply_Int(A=$System.ChainNumber, B=$System.ParticlePerRibbon)
    Use Spawn Probability=false
    UseLoopCountLimit=false
   )
  ]
  particleSpawnScript [
   &InitializeParticle(Lifetime=5, Sprite Size Max=<20 20>, Sprite Size Mode=::NewEnumerator2, UsePositionOffset=false, Write Material Random=true)
   &PartitionParticles(
    Number Of Partitions=$System.ChainNumber
    Particles Per Partition:NiagaraInt32=$System.ParticlePerRibbon
    Partition Mode:ENiagara_PartitionMode_v2=::NewEnumerator1
   )
   /SetVariables(Particles.NormalizedChainPosition:NiagaraFloat=$Particles.PartitionParticles.Normalized Index In Partition) {isSetParameters=true, module=SetVariables_#14}
   /SetVariables(Particles.LengthMulti=1) {isSetParameters=true}
  ]
  particleUpdateScript [
   &ParticleState(Kill Particles When Lifetime Has Elapsed=false)
   &SkeletalMeshLocation(
    Bone / Socket Index=$Particles.PartitionParticles.Partition
    Bone Sampling Mode=::NewEnumerator4
    MaskSpawnGroup=false
    Position Sampling=::NewEnumerator2
    Sampled Position Offset=&Lerp_Vector(
     A=&Lerp_Vector(A=<0 -5 2>, B=<0 5 -2>, Alpha=&MakeFloatFromInt(Int=$Particles.PartitionParticles.Partition))
     B=&Lerp_Vector(A=<0 0 0>, B=<0 0 0>, Alpha=&MakeFloatFromInt(Int=$Particles.PartitionParticles.Partition))
     Alpha=$User.ParaGlideLeanAlpha
    )
    Skeletal Mesh=$User.SkeletalMesh
   ) {module=SkeletalMeshLocation001}
   &PendulumSetup(
    Module.Enable Spring Driver=true
    Pendulum Length:NiagaraFloat=&Lerp_Float(
     A=0
     B=&Multiply_Float(A=$System.ChainLength, B=$Particles.Initial.LengthMulti)
     Alpha=&FloatFromCurve(
      FloatCurve=DataInterface() {propertyValues {Curve {keys [<at 0 0>, {arriveTangent=-0.0703, interpMode=RCIM_Cubic, leaveTangent=-0.0703, tangentMode=RCTM_User, time=1, value=1}]}}}
      CurveIndex=$Particles.NormalizedChainPosition
      Scale Curve=&FloatFromCurve(
       FloatCurve=DataInterface() {propertyValues {Curve {keys [
           {arriveTangent=0.6092, interpMode=RCIM_Cubic, leaveTangent=0.6092, tangentMode=RCTM_User, time=0, value=0}
           {tangentMode=RCTM_User, time=1, value=1}
          ]}}}
       CurveIndex=$Emitter.Age
      )
     )
    )
    Pendulum Pivot:Vector3f=&AddVectorToPosition(
     Position=&AddVectorToPosition(Position=$Particles.SkeletalMeshLocation001.SampledPosition, Vector=<0 0 0>)
     Vector=&TransformVector(Vector=<0 0 0>, Source Space=::World, Destination Space=::Simulation)
    )
    Pendulum Pivot Offset Coordinate Space:ENiagaraCoordinateSpace=::Local
    Pendulum Rest Axis:Vector3f=&TransformVector(
     Vector=&Add_Vector(
      A=&Lerp_Vector(
       A=&Add_Vector(
        A=&Multiply_VectorByFloat(
         Vector=&NormalizeVector(Vector To Normalize=$Engine.Owner.Velocity, Safe Normalize=true, Safe Fallback Threshold=0, Safe Fallback Vector=<0 0 0.001>)
         Float=-0.5
        )
        B=&TransformVector(
         Vector=&RotateVectorByQuaternion(
          Vector To Rotate=&Multiply_VectorByFloat(
           Vector=<0.5 1 0>
           Float=&Lerp_Float(A=1, B=-1, Alpha=&MakeFloatFromInt(Int=$Particles.PartitionParticles.Partition))
          )
          Quaternion=$Particles.SkeletalMeshLocation001.SampledRotation
         )
         Destination Space=::World
        )
       )
       B=&Lerp_Vector(
        A=&Multiply_VectorByFloat(
         Vector=&NormalizeVector(Vector To Normalize=$Engine.Owner.Velocity, Safe Normalize=true, Safe Fallback Threshold=0, Safe Fallback Vector=<0 0 0.001>)
         Float=-1
        )
        B=<0 0 0.1>
        Alpha=$System.SmoothDiveAlpha
       )
       Alpha=&Lerp_Float(
        A=&FloatFromCurve(
         FloatCurve=DataInterface() {propertyValues {Curve {keys [
             {arriveTangent=1.8207, interpMode=RCIM_Cubic, leaveTangent=1.8207, tangentMode=RCTM_User, time=0, value=0}
             {arriveTangent=-0.078, interpMode=RCIM_Cubic, leaveTangent=-0.078, tangentMode=RCTM_User, time=1, value=1}
            ]}}}
         CurveIndex=$Particles.NormalizedChainPosition
        )
        B=1
        Alpha=&FN_FrontEndFloatModifier(Gameplay=0, Front End=1)
       )
      )
      B=&TransformVector(
       Vector=&FN_FrontEndVector3Modifier(Gameplay=<0 0 1>, Front End=&Lerp_Vector(A=<0 -0.1 0.2>, B=<0 -0.2 0.2>, Alpha=$System.SmoothDiveAlpha))
       Destination Space=::World
      )
     )
     Source Space=::World
     Destination Space=::Simulation
    )
    "Spring Driven Constraint (Tightness)"=30
   )
   &SolveForcesAndVelocity(Clamp Velocity=false, Limit Acceleration=false)
   &PendulumConstraint()
   &CalculateAccurateVelocity()
   /SetVariables(Particles.RibbonWidth=&FloatFromCurve(
     FloatCurve=DataInterface() {propertyValues {Curve {keys [
         {arriveTangent=8.0134, interpMode=RCIM_Cubic, leaveTangent=8.0134, tangentMode=RCTM_User, time=0, value=0.7}
         {arriveTangent=0.3471, interpMode=RCIM_Cubic, leaveTangent=0.3471, tangentMode=RCTM_User, time=0.1, value=1.2}
         {arriveTangent=-0.7173, interpMode=RCIM_Cubic, leaveTangent=-0.7173, tangentMode=RCTM_User, time=1, value=2}
        ]}}}
     CurveIndex=$Particles.NormalizedChainPosition
     Scale Curve=&Lerp_Float(A=40, B=30, Alpha=$System.SmoothDiveAlpha)
    )) {isSetParameters=true, module=SetVariables_#15}
   &DynamicMaterialParameters(
    Index 0 Param 1=1
    Index 0 Param 2=&Add_Float(
     A=&Multiply_Float(A=$Engine.System.Age, B=1.2)
     B=&Multiply_FloatByInt(Float=3.14, Integer=$Particles.PartitionParticles.Partition)
    )
    Index 0 Param 3=1
    Index 0 Param 4=&Add_Float(
     A=&Multiply_Float(A=$Engine.System.Age, B=1.2)
     B=&Multiply_FloatByInt(Float=1.6902, Integer=$Particles.PartitionParticles.Partition)
    )
    Index 1 Param 1=1
    Index 1 Param 2=&Add_Float(
     A=&Multiply_Float(A=$Engine.System.Age, B=1.5786)
     B=&Multiply_FloatByInt(Float=1.6902, Integer=$Particles.PartitionParticles.Partition)
    )
    Index 1 Param 3=1
    Index 3 Param 1=1
    Index 3 Param 2=1
    Index 3 Param 3=1
    Module.Index0Param0Write=true
    Module.Index0Param1Write=true
    Module.Index0Param2Write=true
    Module.Index0Param3Write=true
    Module.Index1Param0Write=true
    Module.Index1Param1Write=true
    Module.Index1Param2Write=true
    Module.Index1Param3Write=true
    Module.Index3Param0Write=true
    Module.Index3Param1Write=true
    Module.Index3Param2Write=true
    Module.Index3Param3Write=true
    Write Parameter Index 1=true
    Write Parameter Index 3=true
   )
   &self:Ribbon3Location_03(
    Amplitude:Vector3f=&Multiply_VectorByFloat(
     Vector=&Lerp_Vector(A=<1 1 0>, B=<4 4 0>, Alpha=$User.ParaGlideLeanAlpha)
     Float=&FloatFromCurve(
      FloatCurve=DataInterface() {propertyValues {Curve {keys [
          {interpMode=RCIM_Cubic, tangentMode=RCTM_User, time=0, value=0}
          {arriveTangent=-2.3116, interpMode=RCIM_Cubic, leaveTangent=-2.3116, tangentMode=RCTM_User, time=0.4, value=1}
          <at 1 0>
         ]}}}
      CurveIndex=$Particles.PartitionParticles.Normalized Index In Partition
      Scale Curve=&Lerp_Float(A=1, B=1, Alpha=$System.SmoothDiveAlpha)
     )
    )
    Amplitude2:Vector3f=&Lerp_Vector(A=<5 5 0>, B=<40 40 0>, Alpha=$User.ParaGlideLeanAlpha)
    Amplitude3:Vector3f=&Lerp_Vector(A=<3 3 0>, B=<12 12 0>, Alpha=$User.ParaGlideLeanAlpha)
    Partition:NiagaraFloat=&MakeFloatFromInt(Int=$Particles.PartitionParticles.Partition)
    Scale:Vector3f=&Multiply_VectorByFloat(Vector=<1 1 1>, Float=0.5)
    Scale2:Vector3f=&Multiply_VectorByFloat(Vector=<0.8 0.8 1>, Float=0.5)
    Scale3:Vector3f=&Multiply_VectorByFloat(Vector=<1 1 1>, Float=0.5)
    Time:NiagaraFloat=&Add_Float(
     A=&Multiply_Float(A=$Engine.System.Age, B=-1.5)
     B=$Particles.PartitionParticles.Normalized Index In Partition
    )
   ) {module=Ribbon3Location_03}
   /SetVariables(
    Particles.Color3:LinearColor=<rgba 1.5 1.0535 0 1>
    Particles.DynamicMaterialParameter9:Vector4f=&MakeVector4(
     X=&Multiply_Float(
      A=$Engine.System.Age
      B=&Lerp_Float(A=-2, B=2, Alpha=&MakeFloatFromInt(Int=$Particles.Initial.PartitionParticles.Partition))
     )
     Y=&Lerp_Float(A=-3, B=3, Alpha=&MakeFloatFromInt(Int=$Particles.Initial.PartitionParticles.Partition))
     Z=&Multiply_Float(A=1, B=&Lerp_Float(A=0.4, B=1, Alpha=$System.SmoothDiveAlpha))
     W=&FloatFromCurve(
      FloatCurve=DataInterface() {propertyValues {Curve {keys [{arriveTangent=-2.2554, interpMode=RCIM_Cubic, leaveTangent=-2.2554, tangentMode=RCTM_User, time=0, value=1}, <at 1 0>]}}}
      CurveIndex=$Particles.PartitionParticles.Normalized Index In Partition
     )
    )
    Particles.LightRibbon_Width:NiagaraFloat=&Multiply_Float(
     A=$Particles.RibbonWidth
     B=&FloatFromCurve(
      FloatCurve=DataInterface() {propertyValues {Curve {keys [
          {arriveTangent=0.357, interpMode=RCIM_Cubic, leaveTangent=0.357, tangentMode=RCTM_User, time=0, value=0.3}
          {arriveTangent=-0.0652, interpMode=RCIM_Cubic, leaveTangent=-0.0652, tangentMode=RCTM_User, time=0.2, value=1}
          {arriveTangent=-0.1821, interpMode=RCIM_Cubic, leaveTangent=-0.1821, tangentMode=RCTM_User, time=0.7, value=0}
         ]}}}
      CurveIndex=$Particles.NormalizedChainPosition
      Scale Curve=&Lerp_Float(A=0.9, B=1, Alpha=$System.SmoothDiveAlpha)
     )
    )
   ) {isSetParameters=true, module=SetVariables_#18}
  ]
  rendererClasses [&NiagaraRibbonRendererProperties]
  renderers [{rendererClass=&NiagaraRibbonRendererProperties}]
 }
 NE_LocalSpaceRibbon002 {
  bEnabled=false
  emitterUpdateScript [
   &EmitterState(
    Inactive Response=::NewEnumerator1
    Life Cycle Mode=::NewEnumerator1
    Loop Behavior=::NewEnumerator1
    Loop Duration Mode=::NewEnumerator1
    UseLoopDelay=false
   ) {module=EmitterLifeCycle}
   &SpawnBurst_Instantaneous(
    Spawn Count=&Multiply_Int(A=$System.ChainNumber, B=$System.ParticlePerRibbon)
    Use Spawn Probability=false
    UseLoopCountLimit=false
   )
  ]
  particleSpawnScript [
   &InitializeParticle(Lifetime=5, Sprite Size Max=<20 20>, Sprite Size Mode=::NewEnumerator2, UsePositionOffset=false, Write Material Random=true)
   &PartitionParticles(
    Number Of Partitions=$System.ChainNumber
    Particles Per Partition:NiagaraInt32=$System.ParticlePerRibbon
    Partition Mode:ENiagara_PartitionMode_v2=::NewEnumerator1
   )
   /SetVariables(Particles.NormalizedChainPosition:NiagaraFloat=$Particles.PartitionParticles.Normalized Index In Partition) {isSetParameters=true, module=SetVariables_#14}
   &SkeletalMeshLocation(
    Bone / Socket Index=$Particles.PartitionParticles.Partition
    Bone Sampling Mode=::NewEnumerator4
    MaskSpawnGroup=false
    Sampled Position Offset=&Lerp_Vector(A=<0 0 0>, B=<0 0 150>, Alpha=$Particles.NormalizedChainPosition)
    Sampled Position Offset Coordinate Space:ENiagaraSkelMeshTransforms=::NewEnumerator1
    Skeletal Mesh=$System.Skeletal Mesh
   ) {enabled=false}
  ]
  particleUpdateScript [
   &ParticleState(Kill Particles When Lifetime Has Elapsed=false)
   &SkeletalMeshLocation(
    Bone / Socket Index=$Particles.PartitionParticles.Partition
    Bone Sampling Mode=::NewEnumerator4
    MaskSpawnGroup=false
    Position Sampling=::NewEnumerator2
    Sampled Position Offset=&Lerp_Vector(
     A=&Lerp_Vector(A=<0 -5 2>, B=<0 5 -2>, Alpha=&MakeFloatFromInt(Int=$Particles.PartitionParticles.Partition))
     B=&Lerp_Vector(A=<0 0 0>, B=<0 0 0>, Alpha=&MakeFloatFromInt(Int=$Particles.PartitionParticles.Partition))
     Alpha=$User.ParaGlideLeanAlpha
    )
    Skeletal Mesh=$User.SkeletalMesh
   ) {module=SkeletalMeshLocation001}
   &PendulumSetup(
    Module.Enable Spring Driver=true
    Pendulum Length:NiagaraFloat=&Lerp_Float(
     A=0
     B=$System.ChainLength
     Alpha=&FloatFromCurve(
      FloatCurve=DataInterface() {propertyValues {Curve {keys [<at 0 0>, {arriveTangent=-0.0703, interpMode=RCIM_Cubic, leaveTangent=-0.0703, tangentMode=RCTM_User, time=1, value=1}]}}}
      CurveIndex=$Particles.NormalizedChainPosition
      Scale Curve=&FloatFromCurve(
       FloatCurve=DataInterface() {propertyValues {Curve {keys [
           {arriveTangent=0.6092, interpMode=RCIM_Cubic, leaveTangent=0.6092, tangentMode=RCTM_User, time=0, value=0}
           {tangentMode=RCTM_User, time=1, value=1}
          ]}}}
       CurveIndex=$Emitter.Age
      )
     )
    )
    Pendulum Pivot:Vector3f=&AddVectorToPosition(
     Position=&AddVectorToPosition(Position=$Particles.SkeletalMeshLocation001.SampledPosition, Vector=<0 0 0>)
     Vector=&TransformVector(Vector=<0 0 0>, Source Space=::World, Destination Space=::Simulation)
    )
    Pendulum Pivot Offset Coordinate Space:ENiagaraCoordinateSpace=::Local
    Pendulum Rest Axis:Vector3f=&TransformVector(
     Vector=&Add_Vector(
      A=&Lerp_Vector(
       A=&Add_Vector(
        A=&Multiply_VectorByFloat(
         Vector=&NormalizeVector(Vector To Normalize=$Engine.Owner.Velocity, Safe Normalize=true, Safe Fallback Threshold=0, Safe Fallback Vector=<0 0 0.001>)
         Float=-0.5
        )
        B=&TransformVector(
         Vector=&RotateVectorByQuaternion(
          Vector To Rotate=&Multiply_VectorByFloat(
           Vector=<0.5 1 0>
           Float=&Lerp_Float(A=1, B=-1, Alpha=&MakeFloatFromInt(Int=$Particles.PartitionParticles.Partition))
          )
          Quaternion=$Particles.SkeletalMeshLocation001.SampledRotation
         )
         Destination Space=::World
        )
       )
       B=&Lerp_Vector(
        A=&Multiply_VectorByFloat(
         Vector=&NormalizeVector(Vector To Normalize=$Engine.Owner.Velocity, Safe Normalize=true, Safe Fallback Threshold=0, Safe Fallback Vector=<0 0 0.001>)
         Float=-1
        )
        B=<0 0 0.1>
        Alpha=$System.SmoothDiveAlpha
       )
       Alpha=&Lerp_Float(
        A=&FloatFromCurve(
         FloatCurve=DataInterface() {propertyValues {Curve {keys [
             {arriveTangent=1.8207, interpMode=RCIM_Cubic, leaveTangent=1.8207, tangentMode=RCTM_User, time=0, value=0}
             {arriveTangent=-0.078, interpMode=RCIM_Cubic, leaveTangent=-0.078, tangentMode=RCTM_User, time=1, value=1}
            ]}}}
         CurveIndex=$Particles.NormalizedChainPosition
        )
        B=1
        Alpha=&FN_FrontEndFloatModifier(Gameplay=0, Front End=1)
       )
      )
      B=&TransformVector(
       Vector=&FN_FrontEndVector3Modifier(Gameplay=<0 0 1>, Front End=&Lerp_Vector(A=<0 -0.1 0.2>, B=<0 -0.2 0.2>, Alpha=$System.SmoothDiveAlpha))
       Destination Space=::World
      )
     )
     Source Space=::World
     Destination Space=::Simulation
    )
    "Spring Driven Constraint (Tightness)"=30
   )
   &LinearForce(
    Coordinate Space:ENiagaraCoordinateSpace=::World
    Force=&Multiply_VectorByFloat(Vector=$Engine.Owner.Velocity, Float=-5)
   ) {enabled=false}
   &SolveForcesAndVelocity(Clamp Velocity=false, Limit Acceleration=false)
   &PendulumConstraint()
   &CalculateAccurateVelocity()
   /SetVariables(Particles.RibbonWidth=&FloatFromCurve(
     FloatCurve=DataInterface() {propertyValues {Curve {keys [
         {arriveTangent=8.0134, interpMode=RCIM_Cubic, leaveTangent=8.0134, tangentMode=RCTM_User, time=0, value=0.7}
         {arriveTangent=0.3471, interpMode=RCIM_Cubic, leaveTangent=0.3471, tangentMode=RCTM_User, time=0.1, value=1.2}
         {arriveTangent=-0.7173, interpMode=RCIM_Cubic, leaveTangent=-0.7173, tangentMode=RCTM_User, time=1, value=2}
        ]}}}
     CurveIndex=$Particles.NormalizedChainPosition
     Scale Curve=&Lerp_Float(A=40, B=30, Alpha=$System.SmoothDiveAlpha)
    )) {isSetParameters=true, module=SetVariables_#15}
   &DynamicMaterialParameters(
    Index 0 Param 1=1
    Index 0 Param 2=&Add_Float(
     A=&Multiply_Float(A=$Engine.System.Age, B=1.2)
     B=&Multiply_FloatByInt(Float=3.14, Integer=$Particles.PartitionParticles.Partition)
    )
    Index 0 Param 3=1
    Index 0 Param 4=&Add_Float(
     A=&Multiply_Float(A=$Engine.System.Age, B=1.2)
     B=&Multiply_FloatByInt(Float=1.6902, Integer=$Particles.PartitionParticles.Partition)
    )
    Index 1 Param 1=1
    Index 1 Param 2=&Add_Float(
     A=&Multiply_Float(A=$Engine.System.Age, B=1.5786)
     B=&Multiply_FloatByInt(Float=1.6902, Integer=$Particles.PartitionParticles.Partition)
    )
    Index 1 Param 3=1
    Index 3 Param 1=1
    Index 3 Param 2=1
    Index 3 Param 3=1
    Module.Index0Param0Write=true
    Module.Index0Param1Write=true
    Module.Index0Param2Write=true
    Module.Index0Param3Write=true
    Module.Index1Param0Write=true
    Module.Index1Param1Write=true
    Module.Index1Param2Write=true
    Module.Index1Param3Write=true
    Module.Index3Param0Write=true
    Module.Index3Param1Write=true
    Module.Index3Param2Write=true
    Module.Index3Param3Write=true
    Write Parameter Index 1=true
    Write Parameter Index 3=true
   )
   &self:Ribbon3Location_03(
    Amplitude:Vector3f=&Multiply_VectorByFloat(
     Vector=&Lerp_Vector(A=<1 1 0>, B=<4 4 0>, Alpha=$User.ParaGlideLeanAlpha)
     Float=&FloatFromCurve(
      FloatCurve=DataInterface() {propertyValues {Curve {keys [
          {interpMode=RCIM_Cubic, tangentMode=RCTM_User, time=0, value=0}
          {arriveTangent=-2.3116, interpMode=RCIM_Cubic, leaveTangent=-2.3116, tangentMode=RCTM_User, time=0.4, value=1}
          <at 1 0>
         ]}}}
      CurveIndex=$Particles.PartitionParticles.Normalized Index In Partition
      Scale Curve=&Lerp_Float(A=1, B=1, Alpha=$System.SmoothDiveAlpha)
     )
    )
    Amplitude2:Vector3f=&Lerp_Vector(A=<5 5 0>, B=<40 40 0>, Alpha=$User.ParaGlideLeanAlpha)
    Amplitude3:Vector3f=&Lerp_Vector(A=<3 3 0>, B=<12 12 0>, Alpha=$User.ParaGlideLeanAlpha)
    Partition:NiagaraFloat=&MakeFloatFromInt(Int=$Particles.PartitionParticles.Partition)
    Scale:Vector3f=&Multiply_VectorByFloat(Vector=<1 1 1>, Float=0.5)
    Scale2:Vector3f=&Multiply_VectorByFloat(Vector=<0.8 0.8 1>, Float=0.5)
    Scale3:Vector3f=&Multiply_VectorByFloat(Vector=<1 1 1>, Float=0.5)
    Time:NiagaraFloat=&Add_Float(
     A=&Multiply_Float(A=$Engine.System.Age, B=-1.5)
     B=$Particles.PartitionParticles.Normalized Index In Partition
    )
   ) {module=Ribbon3Location_03}
   /SetVariables(
    Particles.DarkRibbon_Color:LinearColor=&MakeLinearColorFromVectorAndFloat(
     "Vector (RGB)"=&Multiply_VectorByFloat(
      Vector=&MakeVectorFromLinearColorRGB(Color=&ColorFromCurve(
        ColorCurve=DataInterface() {propertyValues {
          AlphaCurve {keys [<at 0 1 RCIM_Cubic>,<at 1 1 RCIM_Cubic>]}
          BlueCurve {keys [<at 0 1.3>,<at 0.2787 2>,<at 0.4745 1.5>,<at 0.7559 0.1673>,<at 0.8762 0>,<at 1 0>]}
          GreenCurve {keys [<at 0 0.6828>,<at 0.2787 0.2779>,<at 0.4745 0>,<at 0.7559 0>,<at 0.8762 0.5146>,<at 1 0.8811>]}
          RedCurve {keys [<at 0 0>,<at 0.2787 0>,<at 0.4745 0.7049>,<at 0.7559 2>,<at 0.8762 2>,<at 1 2>]}
         }}
        CurveIndex=&Waveform(
         Global Input Position=&Multiply_FloatByInt(Float=$Particles.NormalizedChainPosition, Integer=1)
         Global Phase=&Add_Float(
          A=&Multiply_Float(A=$Engine.System.Age, B=-1)
          B=&Multiply_FloatByInt(Float=34, Integer=$Particles.PartitionParticles.Partition)
         )
         "Waveform [1]"=::NewEnumerator12
         "[1] Frequency"=0.8094
         "[1] Amplitude Min/Max"=<0 1>
        )
       ))
      Float=&Lerp_Float(A=0.4, B=&FN_FrontEndFloatModifier(Gameplay=0.9, Front End=0.5), Alpha=$System.SmoothDiveAlpha)
     )
     "Float (Alpha)"=&FloatFromCurve(
      FloatCurve=DataInterface() {propertyValues {Curve {keys [
          <at 0 1 RCIM_Cubic>
          {arriveTangent=-2.1381, interpMode=RCIM_Cubic, leaveTangent=-2.1381, tangentMode=RCTM_User, time=0.4255, value=0.5}
          <at 0.949 0 RCIM_Cubic>
         ]}}}
      CurveIndex=$Particles.NormalizedChainPosition
      Scale Curve=&FloatFromCurve(
       FloatCurve=DataInterface() {propertyValues {Curve {keys [{interpMode=RCIM_Cubic, tangentMode=RCTM_User, time=0, value=0}, {interpMode=RCIM_Cubic, tangentMode=RCTM_User, time=1, value=1}]}}}
       CurveIndex=$Engine.System.Age
       Scale Curve=2.5
      )
     )
    )
    Particles.DarkRibbon_Width:NiagaraFloat=&Multiply_Float(A=$Particles.RibbonWidth, B=1)
    Particles.DynamicMaterialParameter8:Vector4f=&MakeVector4(X=1, Y=1, Z=&Lerp_Float(A=1.5, B=2, Alpha=$System.SmoothDiveAlpha), W=0)
    Particles.RibbonFacing=&CrossProduct(Vector A:Vector3f=$Transient.PhysicsPendulumAxis, Vector B:Vector3f=<0 0 1>, Normalize Incoming Vectors:NiagaraBool=true)
   ) {isSetParameters=true, module=SetVariables_#16}
   /SetVariables(
    Particles.Color3:LinearColor=<rgba 1.5 1.0535 0 1>
    Particles.DynamicMaterialParameter9:Vector4f=&MakeVector4(
     X=&Multiply_Float(
      A=$Engine.System.Age
      B=&Lerp_Float(A=-2, B=2, Alpha=&MakeFloatFromInt(Int=$Particles.Initial.PartitionParticles.Partition))
     )
     Y=&Lerp_Float(A=-3, B=3, Alpha=&MakeFloatFromInt(Int=$Particles.Initial.PartitionParticles.Partition))
     Z=&Multiply_Float(A=1, B=&Lerp_Float(A=0.4, B=1, Alpha=$System.SmoothDiveAlpha))
     W=&FloatFromCurve(
      FloatCurve=DataInterface() {propertyValues {Curve {keys [{arriveTangent=-2.2554, interpMode=RCIM_Cubic, leaveTangent=-2.2554, tangentMode=RCTM_User, time=0, value=1}, <at 1 0>]}}}
      CurveIndex=$Particles.PartitionParticles.Normalized Index In Partition
     )
    )
    Particles.LightRibbon_Width:NiagaraFloat=&Multiply_Float(
     A=$Particles.RibbonWidth
     B=&FloatFromCurve(
      FloatCurve=DataInterface() {propertyValues {Curve {keys [
          {arriveTangent=0.357, interpMode=RCIM_Cubic, leaveTangent=0.357, tangentMode=RCTM_User, time=0, value=0.3}
          {arriveTangent=-0.0652, interpMode=RCIM_Cubic, leaveTangent=-0.0652, tangentMode=RCTM_User, time=0.2, value=1}
          {arriveTangent=-0.1821, interpMode=RCIM_Cubic, leaveTangent=-0.1821, tangentMode=RCTM_User, time=0.7, value=0}
         ]}}}
      CurveIndex=$Particles.NormalizedChainPosition
      Scale Curve=&Lerp_Float(A=0.9, B=1, Alpha=$System.SmoothDiveAlpha)
     )
    )
   ) {isSetParameters=true, module=SetVariables_#18}
   /SetVariables(
    Particles.DynamicMaterialParameter14:Vector4f=&MakeVector4(
     X=&FloatFromCurve(
      FloatCurve=DataInterface() {propertyValues {Curve {keys [
          <at 0 0.05 RCIM_Cubic>
          {arriveTangent=-0.1056, interpMode=RCIM_Cubic, leaveTangent=-0.1056, tangentMode=RCTM_User, time=0.0615, value=1}
          {arriveTangent=-2.6928, tangentMode=RCTM_User, time=1, value=0}
         ]}}}
      CurveIndex=$Particles.PartitionParticles.Normalized Index In Partition
      Scale Curve=&Lerp_Float(A=0.7, B=0.83, Alpha=$System.SmoothDiveAlpha)
     )
     Y=0
     Z=0
     W=0
    )
    Particles.GoldRibbbon_Width:NiagaraFloat=&Multiply_Float(
     A=$Particles.RibbonWidth
     B=&FloatFromCurve(
      FloatCurve=DataInterface() {propertyValues {Curve {keys [
          {arriveTangent=0.357, interpMode=RCIM_Cubic, leaveTangent=0.357, tangentMode=RCTM_User, time=0, value=0.3}
          {arriveTangent=-0.0652, interpMode=RCIM_Cubic, leaveTangent=-0.0652, tangentMode=RCTM_User, time=0.2, value=1}
          {arriveTangent=-0.1821, interpMode=RCIM_Cubic, leaveTangent=-0.1821, tangentMode=RCTM_User, time=0.7, value=0}
         ]}}}
      CurveIndex=$Particles.NormalizedChainPosition
      Scale Curve=&Lerp_Float(A=0.2, B=0.4, Alpha=$System.SmoothDiveAlpha)
     )
    )
   ) {isSetParameters=true, module=SetVariables_#17}
  ]
  rendererClasses [&NiagaraRibbonRendererProperties]
  renderers [
   {rendererClass=&NiagaraRibbonRendererProperties}
   {rendererClass=&NiagaraRibbonRendererProperties, rendererIndex=1}
   {rendererClass=&NiagaraRibbonRendererProperties, rendererIndex=2}
  ]
 }
 NE_SplineThicken002 {
  emitterUpdateScript [
   &EmitterState(Life Cycle Mode=::NewEnumerator1, Loop Behavior=::NewEnumerator1, Loop Duration Mode=::NewEnumerator1, UseLoopDelay=false) {module=EmitterLifeCycle}
   &SpawnRate(
    Spawn Probability=0.8
    SpawnRate=&Lerp_Float(A=0, B=20, Alpha=$System.SmoothDiveAlpha)
    Use Spawn Probability=true
   )
   &SpawnRate(SpawnGroup=2, SpawnRate=20, Use Spawn Probability=false) {module=SpawnRate002}
  ]
  particleSpawnScript [
   &InitializeParticle(
    Color Mode=::NewEnumerator0
    Lifetime=&Lerp_Float(A=1, B=2, Alpha=$System.SmoothDiveAlpha)
    Sprite Size Mode=::NewEnumerator3
    Uniform Sprite Size=&FloatFromCurve(
     FloatCurve=DataInterface() {propertyValues {Curve {keys [<at 0 15 RCIM_Cubic>, {arriveTangent=31.0503, tangentMode=RCTM_User, time=1, value=30}]}}}
     CurveIndex=&RandomRangeFloat(Minimum=0, Maximum=1)
    )
    UsePositionOffset=false
    Write Material Random=true
   )
   &FN_SetSpawnGroup(Number of Spawn Groups=2)
   &PartitionParticles()
   /SetVariables(
    Particles.StarsColor:LinearColor=<rgba 5 1.638 0.7975 1>
    Particles.StarsSpriteSize:Vector2f=&Vector2DFromFloat(Value=&Multiply_Float(A=&RandomRangeFloat(Minimum=1, Maximum=4), B=2))
   ) {isSetParameters=true}
   /SetVariables(Particles.SpriteIndex:NiagaraFloat=&RandomRangeFloat(Minimum=0, Maximum=3)) {isSetParameters=true}
   /SetVariables(Particles.DustSize:Vector2f=&Vector2DFromFloat(Value=&Multiply_Float(
      A=&RandomRangeFloat(Minimum=28, Maximum=36)
      B=&Lerp_Float(A=1.9518, B=1.8, Alpha=$User.ParaGlideLeanAlpha)
     ))) {isSetParameters=true}
   /SetVariables(Particles.ParticleSideMask:NiagaraInt32=&ModuloInt(A=&ReturnParticleIDAttributes(), B=2)) {isSetParameters=true}
   &AddVelocity(
    Velocity=&NormalizeVector(
     Vector To Normalize=$System.FN_Contrail_Setup.Direction Velocity Raw
     Safe Normalize=true
     Safe Fallback Threshold=0
     Safe Fallback Vector=<0 0 0.001>
    )
    Velocity Speed Scale=&RandomRangeFloat(Minimum=100, Maximum=200)
   )
   &SkeletalMeshLocation(
    Bone / Socket Index=$Particles.PartitionParticles.Partition
    Bone Sampling Mode=::NewEnumerator4
    MaskSpawnGroup=false
    Sampled Position Offset=&RandomVector(Vector Scale=10)
    Skeletal Mesh=$User.SkeletalMesh
   )
  ]
  particleUpdateScript [
   &ParticleState()
   &SpriteRotationRate(Rotation Rate=&Multiply_Float(
     A=&UniformAOrBFloat(A=-1, Evaluation Type=::NewEnumerator0, Recalculate Random Each Loop=true, Override Seed=false)
     B=&RandomRangeFloat(Minimum=100, Maximum=250)
    )) {enabled=false, module=SpriteRotationRate001}
   &ScaleSpriteSize(
    Uniform Curve Scale=&RandomRangeFloat(Minimum=0.1, Maximum=0.3)
    Uniform Curve Sprite Scale=DataInterface() {propertyValues {Curve {keys [
        {arriveTangent=6.8926, interpMode=RCIM_Cubic, leaveTangent=6.8926, tangentMode=RCTM_User, time=0, value=0}
        {interpMode=RCIM_Cubic, tangentMode=RCTM_User, time=0.2, value=1}
        {arriveTangent=-3.4257, tangentMode=RCTM_User, time=1, value=0}
       ]}}}
   )
   /SetVariables(Particles.Color=&MultiplyLinearColor(
     Linear Color=$Particles.RibbonColor
     Scale Factor=&MakeVector4_FromLinearColor(LinearColor=&ColorFromCurve(ColorCurve=DataInterface() {propertyValues {
         AlphaCurve {keys [<at 0.0069 0>,<at 0.1431 1>,<at 1 0 RCIM_Cubic>]}
         BlueCurve {keys [<at 0.0035 0>,<at 1 0 RCIM_Cubic>]}
         GreenCurve {keys [<at 0.0035 5.3522>,<at 1 0 RCIM_Cubic>]}
         RedCurve {keys [<at 0.0035 20>,<at 1 0 RCIM_Cubic>]}
        }}))
    )) {isSetParameters=true}
   &DynamicMaterialParameters(
    Index 0 Param 1=&RandomRangeFloat(Minimum=5, Maximum=20)
    Index 0 Param 2=1
    Index 0 Param 3=&Multiply_Float(A=$Engine.System.Age, B=1)
    Module.Index0Param0Write=true
    Module.Index0Param1Write=true
    Module.Index0Param2Write=true
    Module.Index0Param3Write=true
   )
   &CameraOffset(Camera Offset Amount=20)
   /SetVariables(
    Particles.StarsColorUpdate:LinearColor=&MakeLinearColorFromVectorAndFloat(
     "Vector (RGB)"=&MakeVectorFromLinearColorRGB(Color=&MultiplyLinearColor(
       Linear Color=$Particles.StarsColor
       Scale Factor=&MakeVector4_FromLinearColor(LinearColor=&ColorFromCurve(ColorCurve=DataInterface() {propertyValues {
           AlphaCurve {keys [<at 0 1>,<at 1 0>]}
           BlueCurve {keys [<at 0 5>,<at 0.1051 1>,<at 0.2645 0.9531>,<at 0.517 0.7>]}
           GreenCurve {keys [<at 0 5>,<at 0.1051 1>,<at 0.2645 0.9531>,<at 0.517 0.7>]}
           RedCurve {keys [<at 0 5>,<at 0.1051 1>,<at 0.2645 0.9531>,<at 0.517 0.7>]}
          }}))
      ))
     "Float (Alpha)"=&FNTimeOfDayScalar(Night Scale=0.6, Morning Scale=0.8, Day Scale=0.8, Evening Scale=0.8)
    )
    Particles.StarsSizeUpdate:Vector2f=&Multiply_Vector2D_ByFloat(
     Float=&FloatFromCurve(
      FloatCurve=DataInterface() {propertyValues {Curve {keys [
          {arriveTangent=14.26, interpMode=RCIM_Cubic, leaveTangent=14.26, tangentMode=RCTM_User, time=0, value=0}
          {arriveTangentWeight=0.0869, interpMode=RCIM_Cubic, leaveTangentWeight=0.1958, tangentMode=RCTM_User, time=0.2, value=1}
          <at 1 0>
         ]}}}
      Scale Curve=&RandomRangeFloat(Minimum=0.5, Maximum=1.5)
     )
     Vector2D=$Particles.StarsSpriteSize
    )
   ) {isSetParameters=true}
   /SetVariables(
    Particles.StarsGlowColor:LinearColor=&Multiply_LinearColorByFloat(
     Linear Color=&MultiplyLinearColor(Linear Color=$Particles.StarsColorUpdate, Scale Factor=&MakeVector4(X=1, Y=1, Z=1, W=0.1))
     Float=&FNTimeOfDayScalar(Night Scale=0.6, Morning Scale=0.8, Day Scale=0.8, Evening Scale=0.8)
    )
    Particles.StarsGlowSize:Vector2f=&Multiply_Vector2D_ByFloat(Float=7, Vector2D=$Particles.StarsSizeUpdate)
   ) {isSetParameters=true}
   /SetVariables(
    Particles.DustColor:LinearColor=&MakeLinearColorFromVectorAndFloat(
     "Vector (RGB)"=&MakeVectorFromLinearColorRGB(Color=<rgba 0.4065 0.0006 0.7188 0.0261>)
     "Float (Alpha)"=&FloatFromCurve(
      FloatCurve=DataInterface() {propertyValues {Curve {keys [<at 0 0 RCIM_Cubic>, {interpMode=RCIM_Cubic, tangentMode=RCTM_Break, time=0.2, value=1}, <at 1 0>]}}}
      Scale Curve=&Multiply_Float(A=&FN_FrontEndFloatModifier(Gameplay=1.5, Front End=0.3), B=0.5)
     )
    )
    Particles.DustRotation:NiagaraFloat=&RandomRangeFloat(Minimum=0, Maximum=360)
    Particles.DustSizeUpdate:Vector2f=&Multiply_Vector2D_ByFloat(
     Float=&FloatFromCurve(
      FloatCurve=DataInterface() {propertyValues {Curve {keys [
          {arriveTangent=3.5003, interpMode=RCIM_Cubic, leaveTangent=3.5003, tangentMode=RCTM_User, time=0, value=0.5}
          {arriveTangent=0.5833, interpMode=RCIM_Cubic, leaveTangent=0.5833, time=0.18, value=0.9}
          <at 1.2 1.2 RCIM_Cubic>
         ]}}}
      Scale Curve=5
     )
     Vector2D=$Particles.DustSize
    )
   ) {isSetParameters=true}
   &VectorNoiseForce(Force Amount=250, Module.Use Random Noise Time Interval=false)
   &SolveForcesAndVelocity(Clamp Velocity=false, Limit Acceleration=false)
  ]
  rendererClasses [&NiagaraSpriteRendererProperties]
  renderers [
   {rendererClass=&NiagaraSpriteRendererProperties}
   {rendererClass=&NiagaraSpriteRendererProperties, rendererIndex=1}
   {rendererClass=&NiagaraSpriteRendererProperties, rendererIndex=2}
  ]
 }
 NebulaHandSmoke_test {
  bEnabled=false
  emitterUpdateScript [
   &EmitterState()
   &SpawnRate(SpawnRate=15, Use Spawn Probability=false)
   &SpawnRate(SpawnGroup=1, SpawnRate=10, Use Spawn Probability=false) {module=SpawnRate001}
  ]
  particleSpawnScript [
   &InitializeParticle(
    Color=&RandomRangeLinearColor(Minimum=<rgba 0.1253 0 0.7 1.5>, Maximum=<rgba 0.3149 0.0966 1 2>)
    Lifetime=&UniformRangedFloat(Minimum=0.5, Maximum=0.75)
    Sprite Rotation Mode=::NewEnumerator2
    Sprite Size Mode=::NewEnumerator3
    Uniform Sprite Size=&Multiply_Float(
     A=&RandomRangeFloat(Minimum=20, Maximum=40)
     B=&MaskFloatBySpawnGroup(Masked Float=1, Passthrough Float=&RandomRangeFloat(Minimum=0.4, Maximum=0.6))
    )
    UsePositionOffset=false
    Write Material Random=true
   )
   &PartitionParticles(Write Partition to Ribbon ID=::NewEnumerator0)
   &DynamicMaterialParameters(
    Index 0 Param 1=&RandomRangeFloat(Minimum=-0.05, Maximum=0.05)
    Index 0 Param 2=&RandomRangeFloat(Minimum=0, Maximum=32)
    Module.Index0Param0Write=true
    Module.Index0Param1Write=true
    Module.Index0Param2Write=false
    Module.Index0Param3Write=false
   ) {enabled=false}
   &AddVelocity(
    Velocity=&FN_FrontEndVector3Modifier(
     Gameplay=&Lerp_Vector(
      A=&FN_FrontEndVector3Modifier(Gameplay=&MaskVectorBySpawnGroup(Masked Vector=<0 75 100>, Passthrough Vector=<0 50 100>), Front End=<0 0 100>)
      B=&MakeVector(
       X=&SelectFloatFromArray(
        Direct Array Index=$Particles.PartitionParticles.Partition
        Float Selection Array=DataInterface() {propertyValues {FloatData [50,-50,25,-25]}}
       )
       Y=75
       Z=15
      )
      Alpha=$System.Dive
     )
     Front End=&Lerp_Vector(A=<0 -25 100>, B=<0 -150 150>, Alpha=$System.Dive)
    )
    Velocity Speed Scale=&Multiply_Float(
     A=&MaskFloatBySpawnGroup(Masked Float=1, Passthrough Float=&RandomRangeFloat(Minimum=1, Maximum=1.5))
     B=&FN_FrontEndFloatModifier(Gameplay=&MaskFloatBySpawnGroup(Masked Float=0.5, Passthrough Float=0.2), Front End=1)
    )
   )
  ]
  particleUpdateScript [
   &ParticleState()
   &SocketLocation(
    SamplePosOffset=true
    Sampled Position Offset=&FN_FrontEndVector3Modifier(
     Gameplay=<0 0 0>
     Front End=&Lerp_Vector(
      A=&SelectVectorFromArray(
       Direct Array Index=$Particles.PartitionParticles.Partition
       Vector Selection Array=DataInterface() {propertyValues {FloatData [<-10 0 0>, <10 0 0>, <10 0 0>, <-10 0 0>]}}
      )
      B=&SelectVectorFromArray(
       Direct Array Index=$Particles.PartitionParticles.Partition
       Vector Selection Array=DataInterface() {propertyValues {FloatData [<-10 0 0>, <10 0 0>, <15 0 0>, <-15 0 0>]}}
      )
      Alpha=$System.Dive
     )
    )
    Socket Index=$Particles.PartitionParticles.Partition
    Socket Reader=$System.Socket Reader
    Socket Sampling Mode=::NewEnumerator3
    VelocitySpeedLimit=false
   )
   &SolveForcesAndVelocity(Clamp Velocity=false, Limit Acceleration=false)
   &ScaleColor(
    Scale Alpha=&FloatFromCurve(
     FloatCurve=DataInterface() {propertyValues {Curve {keys [
         {arriveTangent=-0.4774, interpMode=RCIM_Cubic, leaveTangent=-0.4774, tangentMode=RCTM_User, time=0, value=0}
         <at 0.5023 0.5898 RCIM_Cubic>
         <at 1.0011 0 RCIM_Cubic>
        ]}}}
     Scale Curve=&FN_FrontEndFloatModifier(Gameplay=&MaskFloatBySpawnGroup(Masked Float=0.5, Passthrough Float=1.5), Front End=1)
    )
    Scale RGB=&Multiply_Vector(
     A=&MakeVectorFromLinearColorRGB(Color=&FN_Select_Color_from_Time_Of_Day(
       Use Manual Time of Day Thresholds=true
       Morning=<rgba 0.942 0.4238 2.5 1>
       Day=<rgba 1 1 1 1>
       Evening=<rgba 1.1427 0.821 2 1>
       Night=<rgba 0 0.3541 1 1>
       Evening Threshold=22
      ))
     B=&VectorFromCurve(
      VectorCurve=DataInterface() {propertyValues {XCurve {keys [<at 0 1>,<at 1 0.9>]}, YCurve {keys [<at 0 1>,<at 1 0.9>]}, ZCurve {keys [<at 0 1>,<at 1 2>]}}}
      Scale Curve=&Multiply_VectorByFloat(
       Vector=&Multiply_VectorByFloat(Vector=<15 15 15>, Float=&MaskFloatBySpawnGroup(Masked Float=1, Passthrough Float=2))
       Float=&Lerp_Float(A=1, B=2, Alpha=$System.SmoothedAudioScale)
      )
     )
    )
    ScaleA=true
    ScaleRGB=true
   ) {enabled=false}
   &ScaleSpriteSize(
    Uniform Curve Scale=&Multiply_Float(
     A=&SelectFloatFromArray(
      Direct Array Index=$Particles.PartitionParticles.Partition
      Float Selection Array=DataInterface() {propertyValues {FloatData [1.75,1.75,0.8,0.8]}}
     )
     B=&Multiply_Float(
      A=&Lerp_Float(A=1, B=1.25, Alpha=$System.SmoothedAudioScale)
      B=&FN_FrontEndFloatModifier(Gameplay=&MaskFloatBySpawnGroup(Masked Float=0.8, Passthrough Float=0.8), Front End=0.85)
     )
    )
    Uniform Curve Sprite Scale=DataInterface() {propertyValues {Curve {keys [
        {arriveTangent=2.3771, interpMode=RCIM_Cubic, leaveTangent=2.3771, tangentMode=RCTM_User, time=0, value=0}
        {interpMode=RCIM_Cubic, leaveTangent=-0.4663, time=1, value=1}
       ]}}}
   )
   &CameraOffset(Camera Offset Amount=50)
   &InterpolateOverTime(
    Rate Of Change:NiagaraFloat=&MaskFloatBySpawnGroup(
     Masked Float=&RandomRangeFloat(Minimum=2, Maximum=3)
     Passthrough Float=&RandomRangeFloat(Minimum=12, Maximum=24)
    )
    Target Position Value:NiagaraPosition=$Particles.Position
    Variable Type=::NewEnumerator6
   )
   /SetVariables(Particles.Position=$Particles.InterpolateOverTime.InterpolateOverTime_Position.Moving Average) {isSetParameters=true}
   &KillParticlesInVolume(Invert Volume=true, Kill Volume Enabled=false, Sphere Radius=25, Volume Origin:Vector3f=$Output.SocketLocation.SampledPosition) {enabled=false}
   &Timeline(
    Play=&SetBoolByFloatComparison(
     A=&NormalizeDistanceRange(
      Start Position:NiagaraPosition=$Output.SocketLocation.SampledPosition
      End Position:NiagaraPosition=$Particles.Position
      Distance:NiagaraFloat=25
      Invert Normalized Range:NiagaraBool=false
     )
     B=0.25
     Comparison Type=::NewEnumerator2
    )
    Play Rate=2.5
    Rewind Rate=0.2
   )
   &ScaleColor(Scale Alpha=$Particles.Timeline.InverseNormalizedTime, ScaleA=true, ScaleRGB=false) {module=ScaleColor001}
  ]
  rendererClasses [&NiagaraSpriteRendererProperties]
  renderers [{rendererClass=&NiagaraSpriteRendererProperties}]
 }
 NebulaRibbons_test {
  bEnabled=false
  emitterUpdateScript [
   &EmitterState()
   &SpawnRate(SpawnRate=120, Use Spawn Probability=false)
   &SpawnRate(SpawnGroup=1, SpawnRate=120, Use Spawn Probability=false) {enabled=false, module=SpawnRate001}
  ]
  particleSpawnScript [
   /SetVariables(Particles.SpawnGroup=$Emitter.SpawnGroup) {enabled=false, isSetParameters=true}
   &PartitionParticles()
   /SetVariables(Particles.UsableIndex:NiagaraInt32=&MergeHelper_IntSwitch(
     0=$Particles.PartitionParticles.Partition
     1=&Add_Integer(A=$Particles.PartitionParticles.Partition, B=2)
     2=?
     3=?
     4=?
     5=?
     Index=$Particles.SpawnGroup
     Num=::NewEnumerator0
    )) {enabled=false, isSetParameters=true}
   &InitializeParticle(
    Lifetime=&MergeHelper_FloatSwitch(0=1, 1=1, 2=?, 3=?, Index=$Particles.PartitionParticles.Partition, Num=::NewEnumerator0)
    UsePositionOffset=false
    Write Material Random=true
   )
   &InitializeRibbon(
    Lifetime=0.82
    Position Mode:ENiagara_PositionInitializationMode=::NewEnumerator0
    Ribbon Facing Mode:ENiagara_UnsetDirectSet=::NewEnumerator1
    Ribbon Facing Vector:Vector3f=<1 0 0>
    Ribbon Width=20
    Write Material Random=true
   )
   &FN_SetSpawnGroup(Number of Spawn Groups=2) {enabled=false}
   &AddVelocity(Velocity=&FN_FrontEndVector3Modifier(
     Gameplay=&Lerp_Vector(
      A=<0 0 250>
      B=&SelectVectorFromArray(
       Direct Array Index=$Particles.PartitionParticles.Partition
       Vector Selection Array=DataInterface() {propertyValues {FloatData [<50 0 400>, <-50 0 400>]}}
      )
      Alpha=$System.Dive
     )
     Front End=&Multiply_VectorByFloat(
      Vector=&NormalizeVector(
       Vector To Normalize=&Lerp_Vector(A=<0 0.2 -1>, B=<0 2 -1>, Alpha=$User.ParaGlideLeanAlpha)
       Safe Normalize=true
       Safe Fallback Threshold=0
       Safe Fallback Vector=<0 0 0.001>
      )
      Float=&Lerp_Float(A=-300, B=-600, Alpha=$User.ParaGlideLeanAlpha)
     )
    ))
   &DynamicMaterialParameters(
    Index 1 Param 4=1
    Module.Index1Float0Write=false
    Module.Index1Float1Write=false
    Module.Index1Float2Write=false
    Module.Index1Float3Write=true
    Write Parameter Index 0=::NewEnumerator0
    Write Parameter Index 1=::NewEnumerator1
   ) {module=DynamicMaterialParameters001}
   &SkeletalMeshLocation(
    Bone / Socket Index=$Particles.PartitionParticles.Partition
    Bone Sampling Mode=::NewEnumerator4
    MaskSpawnGroup=false
    Sampled Position Offset=&SelectVectorFromArray(
     Direct Array Index=$Particles.PartitionParticles.Partition
     Vector Selection Array=DataInterface() {propertyValues {FloatData [<0 -5 0>, <0 5 0>, <-5 0 0>, <5 0 0>]}}
    )
    Skeletal Mesh=$User.SkeletalMesh
   )
  ]
  particleUpdateScript [
   &InheritSourceMovement(
    Applied Position Scale=&Lerp_Vector(A=<1 1 1>, B=<0.99 0.99 0.99>, Alpha=$System.Dive)
    ApplyForce=false
    ApplyPosition=true
    ApplyPositionFalloff=false
   )
   &SolveForcesAndVelocity(Clamp Velocity=false, Limit Acceleration=false)
   &DynamicMaterialParameters(
    Index 0 Param 1=&SelectFloatFromArray(
     Direct Array Index=$Particles.PartitionParticles.Partition
     Float Selection Array=DataInterface() {propertyValues {FloatData [0.8,0.6,0.2,0.1]}}
    )
    Index 0 Param 2=$System.Dive
    Index 0 Param 4=&MakeCustomFloatFromBool(
     Bool=&SetBoolByFloatComparison(A=$System.Dive, B=1, Comparison Type=::NewEnumerator4)
     True Float=1
     False Float=0.8
    )
    Module.Index0Float0Write=true
    Module.Index0Float1Write=true
    Module.Index0Float2Write=false
    Module.Index0Float3Write=true
   )
   &ParticleState()
   &ScaleColor(
    Linear Color Curve=DataInterface() {propertyValues {
      AlphaCurve {keys [<at 0.719 1>,<at 1 0>]}
      BlueCurve {keys [<at 0 0.5>,<at 0.2821 0.4>,<at 0.7519 0.2016>]}
      GreenCurve {keys [<at 0 0.1943>,<at 0.2821 0.0698>,<at 0.7519 0.0039>]}
      RedCurve {keys [<at 0 0.4245>,<at 0.2821 0.2131>,<at 0.7519 0.1648>]}
     }}
    Scale Mode=::NewEnumerator2
   )
  ]
  rendererClasses [&NiagaraRibbonRendererProperties]
  renderers [
   {rendererClass=&NiagaraRibbonRendererProperties}
   {rendererClass=&NiagaraRibbonRendererProperties, rendererIndex=1}
  ]
 }
 Pentagram {
  emitterSpawnScript [&FN_PreserveSkydiveUserVariables() {module=PreserveSkydiveUserVariables}]
  emitterUpdateScript [
   &EmitterState(
    Life Cycle Mode=::NewEnumerator1
    Loop Behavior=::NewEnumerator1
    Loop Duration Mode=::NewEnumerator1
    Reset Age On Awaken=true
    Scalability Mode=::NewEnumerator1
    UseLoopDelay=false
   )
   &SpawnBurst_Instantaneous(Spawn Count=1, Use Spawn Probability=false)
  ]
  particleSpawnScript [
   &FN_SetSpawnGroup(Number of Spawn Groups=1) {module=FN_SetSpawnGroup001}
   &InitializeParticle(
    Color Mode=::NewEnumerator0
    Mesh Scale=<1.3 1.3 0.6>
    Mesh Scale Mode=::NewEnumerator4
    Sprite Size Mode=::NewEnumerator3
    Uniform Sprite Size=400
    UsePositionOffset=false
    Write Material Random=true
   )
   &InitialMeshOrientation(
    Mesh Orientation Mode=::NewEnumerator2
    Orientation Coordinate Space=::Local
    Orientation Vector=<0 0 1>
    Rotation=<1 0 0>
    Rotation Coordinate Space=::NewEnumerator2
    Use Rotation Vector=false
   ) {module=InitialMeshOrientation002}
  ]
  particleUpdateScript [
   /SetVariables(Particles.AlphaAll:NiagaraFloat=&Multiply_Float(A=0.8, B=&Lerp_Float(A=0, B=1, Alpha=1))) {isSetParameters=true, module=SetVariables_#12}
   &FN_DivingAttributes_v2(Accumulate Dive=0.1, Accumulate Rest=0.1, ParaGlideLeanAlpha:NiagaraFloat=$User.ParaGlideLeanAlpha) {enabled=false, module=FN_DivingAttributes_v2001}
   &ParticleState(Kill Particles When Lifetime Has Elapsed=false, Let Infinitely Lived Particles Die When Emitter Deactivates=true)
   &ScaleMeshSize(Scale Factor=&VectorFromCurve(
     VectorCurve=DataInterface() {propertyValues {
       XCurve {keys [
         {arriveTangent=4.4809, interpMode=RCIM_Cubic, leaveTangent=4.4809, tangentMode=RCTM_User, time=0.1, value=0}
         <at 0.4 0.475>
         {arriveTangent=0.875, interpMode=RCIM_Cubic, leaveTangent=0.875, time=0.6, value=0.475}
         {tangentMode=RCTM_User, time=1, value=1}
        ]}
       YCurve {keys [
         {arriveTangent=4.4809, interpMode=RCIM_Cubic, leaveTangent=4.4809, tangentMode=RCTM_User, time=0.1, value=0}
         <at 0.4 0.475>
         {arriveTangent=0.875, interpMode=RCIM_Cubic, leaveTangent=0.875, time=0.6, value=0.475}
         {tangentMode=RCTM_User, time=1, value=1}
        ]}
       ZCurve {keys [{tangentMode=RCTM_User, time=0, value=-0.2}, {tangentMode=RCTM_User, time=1, value=1}]}
      }}
     CurveIndex=$Emitter.Age
    ))
   &ScaleMeshSize(Scale Factor=&Multiply_VectorByFloat(
     Vector=&Multiply_VectorByFloat(Vector=<1 1 1>, Float=&FN_FrontEndFloatModifier(Gameplay=1, Front End=0.9))
     Float=&Lerp_Float(A=1, B=1.3, Alpha=&ClampFloat(Float=$User.ParaGlideLeanAlpha))
    )) {module=ScaleMeshSize002}
   &SkeletalMeshLocation(
    Bone / Socket Index=4
    Bone Sampling Mode=::NewEnumerator4
    MaskSpawnGroup=false
    Orientation Sampling=::NewEnumerator1
    Position Sampling=::NewEnumerator1
    Sampled Position Offset=&NormalizeVector(Vector To Normalize=<0 0 0>, Safe Normalize=true, Safe Fallback Threshold=0, Safe Fallback Vector=<0 0 0.001>)
    Skeletal Mesh=$System.Skeletal Mesh
   ) {module=SkeletalMeshLocation001}
   &UpdateMeshOrientation(
    Facing Direction=&Multiply_Vector(A=Linked() {linkedVariable {name=Particles.SkeletalMeshLocation001.SampledPosition, type=NiagaraPosition}}, B=<-1 1 1>)
    Facing Interpolation=true
    Facing Interpolation Rate:NiagaraFloat=&FN_FrontEndFloatModifier(Gameplay=4, Front End=3)
    Orientation Method=::NewEnumerator1
    Override Quat=false
    TransformFacing=false
    TransformUp=false
    Up Direction:Vector3f=<1 0 0>
   ) {enabled=false, module=UpdateMeshOrientation001}
   /SetVariables(Particles.MeshOrientation=&OrientMeshToVector(
     Look At Direction=&FN_FrontEndVector3Modifier(
      Gameplay=&Multiply_VectorByFloat(
       Vector=&NormalizeVector(
        Vector To Normalize=&NegateVector(Vector:Vector3f=&TransformVector(
          Vector=&Multiply_Vector(A=$Engine.Owner.Velocity, B=<0.3 0.3 1>)
          Source Space=::World
          Destination Space=::Simulation
         ))
        Safe Normalize=true
        Safe Fallback Threshold=0
        Safe Fallback Vector=<0 0 0.001>
       )
       Float=-1
      )
      Front End=<0 0.25 1>
     )
     Reference Axis Vector=<1 0 0>
    )) {isSetParameters=true, module=SetVariables_#13}
   &SolveForcesAndVelocity(Clamp Velocity=false, Limit Acceleration=false)
   &UpdateMeshOrientation(
    Rotation Coordinate Space=::NewEnumerator2
    Rotation Rate=&Lerp_Float(A=0.14, B=0.35, Alpha=$System.FN_Integrator001.Value)
    Rotation Vector=<0 0 1>
   ) {enabled=false}
   &ScaleColor(
    Scale Alpha=&FloatFromCurve(
     FloatCurve=DataInterface() {propertyValues {Curve {keys [
         {arriveTangent=4.0516, interpMode=RCIM_Cubic, leaveTangent=4.0516, tangentMode=RCTM_User, time=0, value=0}
         {arriveTangent=0.0961, interpMode=RCIM_Cubic, leaveTangent=0.0961, tangentMode=RCTM_User, time=0.25, value=1}
         {arriveTangent=0.2272, interpMode=RCIM_Cubic, leaveTangent=0.2272, tangentMode=RCTM_User, time=1, value=1}
        ]}}}
     Scale Curve=&Lerp_Float(A=0.2, B=0.5, Alpha=$System.Dive)
    )
    ScaleA=true
    ScaleRGB=false
   )
   &CameraOffset(Camera Offset Amount=&Lerp_Float(A=0, B=-18, Alpha=$System.FN_Integrator001.Value)) {enabled=false}
   &FN_Mesh_Attributes(Color=<rgba 10 2.7167 0.5 1>, Create Color Attribute=true, Module Output=::NewEnumerator1)
   &FN_Mesh_Attributes(
    Color=&MakeLinearColorFromVectorAndFloat(
     "Vector (RGB)"=&Multiply_Vector(
      A=&MakeVectorFromLinearColorRGB(Color=<rgba 0.0282 0.0049 0.0365 1>)
      B=&Multiply_VectorByFloat(Vector=<1 1 1>, Float=15)
     )
     "Float (Alpha)"=&Multiply_Float(A=1, B=$Particles.AlphaAll)
    )
    Create Color Attribute=true
    Create Scale Attribute=true
    Module Output=::NewEnumerator1
    Scale=&VectorFromCurve(
     VectorCurve=DataInterface() {propertyValues {
       XCurve {keys [
         {arriveTangent=4.4809, interpMode=RCIM_Cubic, leaveTangent=4.4809, tangentMode=RCTM_User, time=0.1, value=0}
         <at 0.4 0.475>
         {arriveTangent=0.875, interpMode=RCIM_Cubic, leaveTangent=0.875, time=0.6, value=0.475}
         {tangentMode=RCTM_User, time=1, value=1}
        ]}
       YCurve {keys [
         {arriveTangent=4.4809, interpMode=RCIM_Cubic, leaveTangent=4.4809, tangentMode=RCTM_User, time=0.1, value=0}
         <at 0.4 0.475>
         {arriveTangent=0.875, interpMode=RCIM_Cubic, leaveTangent=0.875, time=0.6, value=0.475}
         {tangentMode=RCTM_User, time=1, value=1}
        ]}
       ZCurve {keys [{tangentMode=RCTM_User, time=0, value=-0.2}, {tangentMode=RCTM_User, time=1, value=1}]}
      }}
     CurveIndex=$Emitter.Age
     Scale Curve=<3 3 1>
    )
   ) {module=FN_Mesh_Attributes002}
   &DynamicMaterialParameters(
    Index 0 Param 1=3
    Index 0 Param 2=0.4
    Index 2 Param 1=&Lerp_Float(
     A=HlslExpression() {hlslExpression=Particles.Age * 4.0 +5.4}
     B=HlslExpression() {hlslExpression=Particles.Age * 8.0 +5.4}
     Alpha=$System.SmoothDiveAlpha
    )
    Module.Index0Float0Write=true
    Module.Index0Float1Write=true
    Module.Index0Float2Write=true
    Module.Index0Float3Write=true
    Module.Index2Float0Write=true
    Module.Index2Float1Write=false
    Module.Index2Float2Write=false
    Module.Index2Float3Write=false
    Module.Index3Float0Write=false
    Module.Index3Float1Write=false
    Module.Index3Float2Write=false
    Module.Index3Float3Write=false
    Write Parameter Index 2=::NewEnumerator1
    Write Parameter Index 3=::NewEnumerator1
   ) {module=DynamicMaterialParameters001}
   &FN_Light_Attributes(
    Color=<rgba 1 0.2 0 1>
    Create Color Attribute=true
    Create Exponent Attribute=true
    Create Radius Attribute=true
    Exponent=2
    Radius=90
    Scale Color=&FloatFromCurve(
     FloatCurve=DataInterface() {propertyValues {Curve {keys [
         {interpMode=RCIM_Cubic, tangentMode=RCTM_User, time=1.1, value=0}
         <at 1.4 0.25>
         {arriveTangent=2.7482, interpMode=RCIM_Cubic, leaveTangent=2.7482, tangentMode=RCTM_User, time=1.9, value=0.25}
         {interpMode=RCIM_Cubic, tangentMode=RCTM_User, time=2.4, value=1}
        ]}}}
     CurveIndex=$Emitter.Age
     Scale Curve=1.29
    )
   )
  ]
  rendererClasses [&NiagaraLightRendererProperties, &NiagaraMeshRendererProperties]
  renderers [
   {rendererClass=&NiagaraLightRendererProperties, rendererIndex=0}
   {rendererClass=&NiagaraMeshRendererProperties, rendererIndex=1}
   {rendererClass=&NiagaraMeshRendererProperties, rendererIndex=2}
  ]
 }
 Ribbons {
  bEnabled=false
  emitterUpdateScript [
   &EmitterState()
   &SpawnRate(SpawnRate=2, Use Spawn Probability=false)
  ]
  particleSpawnScript [
   &FN_SetSpawnGroup(Number of Spawn Groups=2)
   &PartitionParticles()
   &InitializeParticle(
    Color Mode=::NewEnumerator0
    Lifetime Max=3
    Lifetime Min=2
    Lifetime Mode=::NewEnumerator1
    Mass Mode=::NewEnumerator1
    Mesh Scale Mode=::NewEnumerator3
    Sprite Size Mode=::NewEnumerator4
    UsePositionOffset=false
    Write Material Random=true
   ) {module=InitializeParticle001}
   &InitialMeshOrientation(
    Mesh Orientation Mode=::NewEnumerator2
    Orientation Coordinate Space=::Local
    Rotation=&RandomRangeVector(Minimum=<0 0 0>, Maximum=<0 0 1>)
    Use Rotation Vector=true
   )
  ]
  particleUpdateScript [
   &ParticleState()
   &UpdateMeshOrientation(
    Rotation Rate=&SelectFloatFromArray(
     Direct Array Index=$Particles.Initial.PartitionParticles.Partition
     Float Selection Array=DataInterface() {propertyValues {FloatData [-1,1]}}
    )
    Rotation Vector=<0 0 1>
   )
   &SkeletalMeshLocation(
    Bone / Socket Index=$Particles.SpawnGroup
    Bone Sampling Mode=::NewEnumerator4
    Inherit Velocity=false
    Interpolate Mesh Sampling=false
    MaskSpawnGroup=false
    Sampled Position Offset=&VectorFromCurve(
     VectorCurve=DataInterface() {propertyValues {
       XCurve {keys [<at 0 0>,<at 1 0>]}
       YCurve {keys [<at 0 0>,<at 1 0>]}
       ZCurve {keys [
         {interpMode=RCIM_Cubic, leaveTangent=0.0639, tangentMode=RCTM_Break, time=0, value=0.55}
         {arriveTangent=1.8134, tangentMode=RCTM_User, time=0.9, value=1.35}
        ]}
      }}
     Scale Curve=<1 1 0>
    )
    Skeletal Mesh=$User.SkeletalMesh
   )
   &ScaleMeshSize(Scale Factor=&VectorFromCurve(VectorCurve=DataInterface() {propertyValues {
       XCurve {keys [<at 0 1>,<at 1 1>]}
       YCurve {keys [<at 0 1>,<at 1 1>]}
       ZCurve {keys [{interpMode=RCIM_Cubic, leaveTangent=2.7698, tangentMode=RCTM_Break, time=0, value=1}, <at 0.2 1.375>, <at 1 2>]}
      }}))
   &SolveForcesAndVelocity(Clamp Velocity=false, Limit Acceleration=false)
   &ScaleColor(
    Linear Color Curve=DataInterface() {propertyValues {
      AlphaCurve {keys [<at 0 0>,<at 0.2169 1>,<at 0.4598 1>,<at 1 0.2>]}
      BlueCurve {keys [<at 0.1005 4.0972>,<at 0.386 1>,<at 0.9897 3>]}
      GreenCurve {keys [<at 0.1005 5>,<at 0.386 0.1112>,<at 0.9897 0>]}
      RedCurve {keys [<at 0.1005 0.78>,<at 0.386 0.4056>,<at 0.9897 1.2376>]}
     }}
    Scale Mode=::NewEnumerator2
   ) {enabled=false, module=ScaleColor001}
   &ScaleColor(
    Linear Color Curve=DataInterface() {propertyValues {
      AlphaCurve {keys [<at 0.719 0.5>,<at 1 0>]}
      BlueCurve {keys [<at 0.2821 0.0151>]}
      GreenCurve {keys [<at 0.2821 0.9412>]}
      RedCurve {keys [<at 0.2821 2>]}
     }}
    Scale Mode=::NewEnumerator2
   ) {enabled=false}
  ]
  rendererClasses [&NiagaraMeshRendererProperties]
  renderers [
   {rendererClass=&NiagaraMeshRendererProperties, rendererIndex=0}
   {rendererClass=&NiagaraMeshRendererProperties, rendererIndex=1}
  ]
 }
 StarDust_Trails {
  emitterUpdateScript [
   &EmitterState()
   &SpawnParticlesFromOtherEmitter(
    Attribute Reader=DataInterface() {propertyValues {EmitterBinding {bindingMode=Other, emitterName=Card_Dome001}}}
    Spawn Rate=8
    Spawn Rate Per Particle Cap=80
    Use Maximum Age=false
    Use Minimum Age=false
   )
  ]
  particleSpawnScript [
   &InitializeParticle(
    Color=<rgba 2 1.433 0.251 0.14>
    Lifetime=&RandomRangeFloat(Minimum=0.25, Maximum=0.5)
    Ribbon Width=15
    Ribbon Width Mode=::NewEnumerator1
    Sprite Size Max=<80 80>
    Sprite Size Min=<40 40>
    Sprite Size Mode=::NewEnumerator2
    UsePositionOffset=false
    Write Material Random=true
   )
   &SampleParticlesFromOtherEmitter(Mesh Orientation Sampling:ENiagara_DirectReadSamplingMode=::NewEnumerator0, Particle ID Sampling=::NewEnumerator1)
   &AddVelocity(Velocity=&RandomRangeVector(Minimum=<0 0 100>, Maximum=<0 0 200>))
   /SetVariables(Particles.DustSize:Vector2f=&Vector2DFromFloat(Value=&Multiply_Float(
      A=&RandomRangeFloat(Minimum=35, Maximum=43)
      B=&Lerp_Float(A=1.5, B=3, Alpha=$User.ParaGlideLeanAlpha)
     ))) {isSetParameters=true}
   &CameraOffset(Camera Offset Amount=20) {enabled=false}
  ]
  particleUpdateScript [
   &ParticleState()
   &ScaleVelocity(Velocity Scale=&Lerp_Vector(A=<0 0 1>, B=<0 0 3>, Alpha=$User.ParaGlideLeanAlpha))
   &SolveForcesAndVelocity(Clamp Velocity=false, Limit Acceleration=false)
   &DynamicMaterialParameters(
    Index 0 Param 1=1
    Index 0 Param 2=0.7
    Module.Index0Float0Write=true
    Module.Index0Float1Write=true
    Module.Index0Float2Write=true
    Module.Index0Float3Write=true
   )
   &ScaleColor(ScaleA=true, ScaleRGB=true) {enabled=false}
   &ScaleColor(
    Scale Alpha=&FloatFromCurve(
     FloatCurve=DataInterface() {propertyValues {Curve {keys [{interpMode=RCIM_Cubic, leaveTangent=-1.505, tangentMode=RCTM_Break, time=0, value=1}, <at 1 0>]}}}
     Scale Curve=&FN_FrontEndFloatModifier(Gameplay=1, Front End=0.7)
    )
    ScaleA=true
    ScaleRGB=true
   ) {module=ScaleColor001}
   /SetVariables(
    Particles.DustColor:LinearColor=&MakeLinearColorFromVectorAndFloat(
     "Vector (RGB)"=&MakeVectorFromLinearColorRGB(Color=<rgba 0.4065 0.0006 0.7188 0.0261>)
     "Float (Alpha)"=&FloatFromCurve(
      FloatCurve=DataInterface() {propertyValues {Curve {keys [<at 0 0 RCIM_Cubic>, {interpMode=RCIM_Cubic, tangentMode=RCTM_Break, time=0.2, value=1}, <at 1 0>]}}}
      Scale Curve=&Multiply_Float(A=&FN_FrontEndFloatModifier(Gameplay=1.5, Front End=0.3), B=0.5)
     )
    )
    Particles.DustRotation:NiagaraFloat=&RandomRangeFloat(Minimum=0, Maximum=360)
    Particles.DustSizeUpdate:Vector2f=&Multiply_Vector2D_ByFloat(
     Float=&FloatFromCurve(
      FloatCurve=DataInterface() {propertyValues {Curve {keys [
          {arriveTangent=3.5003, interpMode=RCIM_Cubic, leaveTangent=3.5003, tangentMode=RCTM_User, time=0, value=0.5}
          {arriveTangent=0.5833, interpMode=RCIM_Cubic, leaveTangent=0.5833, time=0.18, value=0.9}
          <at 1.2 1.2 RCIM_Cubic>
         ]}}}
      Scale Curve=5
     )
     Vector2D=$Particles.DustSize
    )
   ) {isSetParameters=true}
  ]
  rendererClasses [&NiagaraSpriteRendererProperties]
  renderers [{rendererClass=&NiagaraSpriteRendererProperties}]
 }
 Wave {
  bEnabled=false
  emitterSpawnScript [&FN_PreserveSkydiveUserVariables() {enabled=false, module=PreserveSkydiveUserVariables}]
  emitterUpdateScript [
   &EmitterState()
   &SpawnRate(SpawnRate=2, Use Spawn Probability=false)
  ]
  particleSpawnScript [
   &FN_SetSpawnGroup(Number of Spawn Groups=1) {module=FN_SetSpawnGroup001}
   &InitializeParticle(Color Mode=::NewEnumerator0, Mesh Scale=<1 2 2>, Mesh Scale Mode=::NewEnumerator4, UsePositionOffset=false, Write Material Random=true) {module=InitializeParticle001}
   &InitialMeshOrientation(
    Orientation Axis=<0 0 1>
    Orientation Coordinate Space=::Local
    Orientation Vector=<0 0 1>
    Rotation=&MakeVector(X=&RandomRangeFloat(Minimum=-1, Maximum=1), Y=0, Z=0)
    Use Orientation Vector=true
    Use Rotation Vector=false
   )
   /SetVariables(Particles.Position=&Add_Vector(A=Linked() {linkedVariable {name=Engine.Owner.Position, type=NiagaraPosition}}, B=<0 0 100>)) {enabled=false, isSetParameters=true}
  ]
  particleUpdateScript [
   &ParticleState()
   &ScaleMeshSize(Scale Factor=&Multiply_VectorByFloat(
     Vector=&Multiply_VectorByFloat(Vector=<0.8 1 1>, Float=&FN_FrontEndFloatModifier(Gameplay=1, Front End=0.9))
     Float=&Lerp_Float(A=0.6, B=0.8, Alpha=&ClampFloat(Float=$User.ParaGlideLeanAlpha))
    )) {module=ScaleMeshSize001}
   &SkeletalMeshLocation(
    Bone / Socket Index=4
    Bone Sampling Mode=::NewEnumerator4
    MaskSpawnGroup=false
    Sampled Position Offset=&Lerp_Vector(A=<0 85 0>, B=<120 85 0>, Alpha=$User.ParaGlideLeanAlpha)
    Skeletal Mesh=$System.Skeletal Mesh
   ) {module=SkeletalMeshLocation001}
   &UpdateMeshOrientation(
    Facing Direction=&Multiply_Vector(A=Linked() {linkedVariable {name=Particles.SkeletalMeshLocation001.SampledPosition, type=NiagaraPosition}}, B=<-1 1 1>)
    Facing Interpolation=true
    Facing Interpolation Rate:NiagaraFloat=&FN_FrontEndFloatModifier(Gameplay=4, Front End=3)
    Orientation Method=::NewEnumerator1
    Override Quat=false
    TransformFacing=false
    TransformUp=false
    Up Direction:Vector3f=<1 0 0>
   ) {module=UpdateMeshOrientation001}
   /SetVariables(Particles.MeshOrientation=&OrientMeshToVector(
     Look At Direction=&FN_FrontEndVector3Modifier(
      Gameplay=&Multiply_VectorByFloat(
       Vector=&NormalizeVector(
        Vector To Normalize=&NegateVector(Vector:Vector3f=&TransformVector(
          Vector=&Multiply_Vector(A=$Engine.Owner.Velocity, B=<0.3 0.3 1>)
          Source Space=::World
          Destination Space=::Local
         ))
        Safe Normalize=true
        Safe Fallback Threshold=0
        Safe Fallback Vector=<0 0 0.001>
       )
       Float=1
      )
      Front End=<0 1 -1>
     )
     Reference Axis Vector=<1 0 0>
    )) {isSetParameters=true}
   &SolveForcesAndVelocity(Clamp Velocity=false, Limit Acceleration=false)
   &ScaleColor(
    Scale Alpha=&Multiply_Float(
     A=&FloatFromCurve(
      FloatCurve=DataInterface() {propertyValues {Curve {keys [
          {arriveTangent=4.0516, interpMode=RCIM_Cubic, leaveTangent=4.0516, tangentMode=RCTM_User, time=0, value=0}
          {arriveTangent=0.0961, interpMode=RCIM_Cubic, leaveTangent=0.0961, tangentMode=RCTM_User, time=0.25, value=0.5}
          <at 0.5 0.5>
          {arriveTangent=0.2272, interpMode=RCIM_Cubic, leaveTangent=0.2272, tangentMode=RCTM_User, time=1, value=0}
         ]}}}
      Scale Curve=&Lerp_Float(A=1, B=2, Alpha=&ClampFloat(Float=$User.ParaGlideLeanAlpha))
     )
     B=&FNTimeOfDayScalar(Night Scale=0.3, Morning Scale=1, Evening Scale=1)
    )
    Scale RGB=&MakeVectorFromLinearColorRGB(Color=&ColorFromCurve(ColorCurve=DataInterface() {propertyValues {
        AlphaCurve {keys [<at 0 1 RCIM_Cubic>]}
        BlueCurve {keys [<at 0 1 RCIM_Cubic>,<at 0.2386 1 RCIM_Cubic>]}
        GreenCurve {keys [<at 0 0.3697 RCIM_Cubic>,<at 0.2386 0.2742 RCIM_Cubic>]}
        RedCurve {keys [<at 0 0.1628 RCIM_Cubic>,<at 0.2386 0.6238 RCIM_Cubic>]}
       }}))
    ScaleA=true
    ScaleRGB=true
   ) {enabled=false}
   &SkeletalMeshLocation(
    Bone / Socket Index=5
    Bone Sampling Mode=::NewEnumerator4
    Inherit Velocity=false
    Interpolate Mesh Sampling=false
    MaskSpawnGroup=false
    Sampled Position Offset=&Lerp_Vector(A=<0 0 -85>, B=<1 1 -85>, Alpha=$System.FN_Integrator001.Value)
    Sampled Position Offset Coordinate Space:ENiagaraMeshTransforms=::NewEnumerator2
    Skeletal Mesh=$User.SkeletalMesh
   ) {enabled=false}
  ]
  rendererClasses [&NiagaraMeshRendererProperties]
  renderers [{rendererClass=&NiagaraMeshRendererProperties, rendererIndex=0}]
 }
}
summary {
 systemName=NS_Contrail_Luminence
 emitters [
  {bEnabled=false, emitterName=NebulaRibbons_test, rendererClasses [&NiagaraRibbonRendererProperties]}
  {bEnabled=false, emitterName=NebulaHandSmoke_test, rendererClasses [&NiagaraSpriteRendererProperties]}
  {bEnabled=false, emitterName=Card_Stream, rendererClasses [&NiagaraMeshRendererProperties]}
  {
   bEnabled=false
   emitterName=Center001
   rendererClasses [&NiagaraLightRendererProperties, &NiagaraMeshRendererProperties]
  }
  {
   bEnabled=false
   emitterName=Card_Dome
   rendererClasses [&NiagaraMeshRendererProperties, &NiagaraSpriteRendererProperties]
  }
  {bEnabled=false, emitterName=Wave, rendererClasses [&NiagaraMeshRendererProperties]}
  {bEnabled=false, emitterName=Ribbons, rendererClasses [&NiagaraMeshRendererProperties]}
  {emitterName=NE_SplineThicken002, rendererClasses [&NiagaraSpriteRendererProperties]}
  {emitterName=Flare, rendererClasses [&NiagaraSpriteRendererProperties]}
  {bEnabled=false, emitterName=NE_LocalSpaceRibbon002, rendererClasses [&NiagaraRibbonRendererProperties]}
  {emitterName=Card_Stream001, rendererClasses [&NiagaraMeshRendererProperties]}
  {emitterName=Card_Dome001, rendererClasses [&NiagaraMeshRendererProperties]}
  {emitterName=Pentagram, rendererClasses [&NiagaraLightRendererProperties, &NiagaraMeshRendererProperties]}
  {emitterName=DarkRibbon, rendererClasses [&NiagaraRibbonRendererProperties]}
  {emitterName=LightRibbon, rendererClasses [&NiagaraRibbonRendererProperties]}
  {emitterName=GoldRibbon, rendererClasses [&NiagaraRibbonRendererProperties]}
  {emitterName=StarDust_Trails, rendererClasses [&NiagaraSpriteRendererProperties]}
 ]
 userVariables [
  User.bIsFrontEnd {type=NiagaraBool}
  User.bIsFrontEndPreview {type=NiagaraBool}
  User.ParaGlideLeanAlpha {type=NiagaraFloat, defaultValue=0, description=""}
  User.PlayerSpeed {type=NiagaraFloat, defaultValue=0, description=""}
  User.RibbonIDOffset {type=NiagaraInt32}
  User.Skeletal Mesh01 {
   type=NiagaraDataInterfaceSkeletalMesh
   defaultValue {
    struct=NiagaraExt_VariableValue_DataInterface
    value {dataInterface=&self:NiagaraDataInterfaceSkeletalMesh_3}
   }
  }
  User.Skeletal Mesh02 {
   type=NiagaraDataInterfaceSkeletalMesh
   defaultValue {
    struct=NiagaraExt_VariableValue_DataInterface
    value {dataInterface=&self:NiagaraDataInterfaceSkeletalMesh_2}
   }
  }
  User.SkeletalMesh {
   type=NiagaraDataInterfaceSkeletalMesh
   defaultValue {
    struct=NiagaraExt_VariableValue_DataInterface
    value {dataInterface=&self:NiagaraDataInterfaceSkeletalMesh_0}
   }
  }
  User.SkeletalMesh2 {
   type=NiagaraDataInterfaceSkeletalMesh
   defaultValue {
    struct=NiagaraExt_VariableValue_DataInterface
    value {dataInterface=&self:NiagaraDataInterfaceSkeletalMesh_1}
   }
  }
  User.VariantPicker {type=NiagaraFloat, defaultValue=0, description=""}
 ]
}
